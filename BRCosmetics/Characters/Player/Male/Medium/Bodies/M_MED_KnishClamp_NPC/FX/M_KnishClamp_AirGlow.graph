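# Translucent unlit Niagara sprite and ribbon material for the KnishClamp NPC's air glow effect.
# Particle color, optionally squared via SaturationBlend and blended by the smoke texture's green channel, is scaled by Brightness into an emissive-adjustment function.
# Particle alpha times that mask is depth-faded and multiplied by a camera-to-object distance band, hiding the glow up close and far away.
# UseRedChannel has no effect, as both its inputs read green.
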Material /BRCosmetics/Characters/Player/Male/Medium/Bodies/M_MED_KnishClamp_NPC/FX/M_KnishClamp_AirGlow
graph {
 properties {
  bUsedWithNiagaraRibbons=true
  bUsedWithNiagaraSprites=true
  blendMode=BLEND_Translucent
  pixelDepthOffsetMode=PDOM_Legacy
  shadingModel=MSM_Unlit
  usedShadingModels=Unlit
 }
 nodes [
  n1 = ParticleColor()
  n2 = LinearInterpolate(A=@n1.RGB, B=Multiply(A=@n1.RGB, B=@n1.RGB), Alpha=ScalarParameter() {parameterName=SaturationBlend})
  n3 = TextureSampleParameter2D() {parameterName=BaseTexture, texture=/BRCosmetics/Athena/Effects/Textures/T_InkySmokeTrail}
  n4 = LinearInterpolate(A=@n3.G, B=@n3.G, Alpha=ScalarParameter() {parameterName=UseRedChannel})
  MaterialFunctionCall(Emissive=Multiply(A=LinearInterpolate(A=@n2, B=@n1.RGB, Alpha=@n4), B=ScalarParameter() {defaultValue=1, parameterName=Brightness})) {materialFunction=&MF_FX_EmissiveAdjustments}
  n5 = Distance(A=ObjectPositionWS(), B=CameraPositionWS())
  n6 = ScalarParameter() {defaultValue=2500, group=Knish | Dissolve, parameterName=Knish - Start Fade Distance, sortPriority=44}
  n7 = Saturate(None=Divide(
    A=Subtract(A=@n5, B=@n6)
    B=ScalarParameter() {defaultValue=600, group=Knish | Dissolve, parameterName=Knish - Start Fade Length, sortPriority=44}
   ))
  n8 = Subtract(A=@n5, B=ScalarParameter() {defaultValue=10000, group=Knish | Dissolve, parameterName=Knish - Max Fade Distance, sortPriority=45})
  n9 = ScalarParameter() {defaultValue=2000, group=Knish | Dissolve, parameterName=Knish - Max Fade Length, sortPriority=45}
  n10 = OneMinus(None=Saturate(None=Divide(A=@n8, B=@n9)))
  Multiply(
   A=DepthFade(Opacity=Multiply(A=@n1.A, B=@n4), FadeDistance=ScalarParameter() {defaultValue=20, parameterName=FadeDist})
   B=NamedRerouteDeclaration(None=Multiply(A=@n7, B=@n10)) {name=Distance, nodeColor=<rgba 0 0.8594 1 1>}
  )
 ]
}
