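# Idle Niagara effect for the KnishClamp NPC body, layering sprite glow, blue haze and animated smoke around the character.
# Glow and Smoke re-sample User.SkeletalMesh each update, with Smoke drifting along curve-driven offsets over particle life.
# SmokeLow samples the mesh at spawn, then sinks while partly inheriting source movement; Haze rises from a tapered cylinder.
# User.AlphaMult scales opacity in all four emitters, and time of day dims both smoke layers at night.
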NiagaraSystem /BRCosmetics/Characters/Player/Male/Medium/Bodies/M_MED_KnishClamp_NPC/FX/NS_Character_KnishClamp_Idle
dependencies {
 usedDataInterfaces [
  &NiagaraDataInterfaceCurve
  &NiagaraDataInterfaceColorCurve
  &NiagaraDataInterfaceSpriteRendererInfo
 ]
 usedDynamicInputs [&AngleConversion]
 usedModules [&SystemState]
}
emitters {
 Glow {
  emitterUpdateScript [
   &EmitterState()
   &SpawnRate(SpawnRate=1.5, Use Spawn Probability=false)
  ]
  particleSpawnScript [
   /Niagara/Modules/Spawn/Initialization/InitializeParticle(
    Color=<rgba 5 5 5 0.05>
    Lifetime=&UniformRangedFloat(Minimum=2, Maximum=2.5)
    Module.Write Color=true
    Module.Write Lifetime=true
    Module.Write Mass=false
    Module.Write Material Random=true
    Module.Write Position=false
    Module.Write Scale=false
    Module.Write SpriteRotation=false
    Module.Write SpriteSize=true
    Position=&SimulationPosition()
    Sprite Rotation=&UniformRangedFloat(Minimum=-180, Maximum=180)
    Sprite Size=&Vector2DFromFloat(Value=&UniformRangedFloat(Minimum=30, Maximum=50))
   ) {enabled=false}
   /Niagara/Modules/Spawn/Initialization/V2/InitializeParticle(
    Color=<rgba 5 5 5 0.05>
    Lifetime=&UniformRangedFloat(Minimum=2, Maximum=2.5)
    Sprite Size=&Vector2DFromFloat(Value=&UniformRangedFloat(Minimum=30, Maximum=50))
    Sprite Size Mode=::NewEnumerator4
    UsePositionOffset=false
    Write Material Random=true
   ) {module=InitializeParticle001}
   &CameraOffset(Camera Offset Amount=&RandomRangeFloat(Minimum=-5, Maximum=10))
   &AddVelocity(Velocity=&RandomRangeVector(Minimum=<-1 -1 5>, Maximum=<1 1 2>))
  ]
  particleUpdateScript [
   &SkeletalMeshLocation(
    Bone Sampling Mode=::NewEnumerator4
    Inherit Velocity=false
    MaskSpawnGroup=false
    Sampled Position Offset=&RandomRangeVector(Minimum=<-2 -2 -2>, Maximum=<2 2 2>)
    Skeletal Mesh=$User.SkeletalMesh
   )
   &SolveForcesAndVelocity(Clamp Velocity=false, Limit Acceleration=false)
   &ParticleState()
   &ScaleSpriteSize(Scale Factor=<0.75 2>, Scale Sprite Size Mode=::NewEnumerator2) {module=ScaleSpriteSize001}
   &ScaleSpriteSize(Scale Factor=&Vector2DFromFloat(Value=&FloatFromCurve(
      FloatCurve=DataInterface() {propertyValues {Curve {keys [
          {arriveTangent=7.5764, interpMode=RCIM_Cubic, leaveTangent=7.5765, tangentMode=RCTM_User, time=0, value=0.02}
          {arriveTangent=1.2063, interpMode=RCIM_Cubic, leaveTangent=1.2063, tangentMode=RCTM_User, time=0.18, value=0.99}
          <at 1 0.5 RCIM_Cubic>
         ]}}}
      Scale Curve=3
     )))
   &ScaleColor(
    Scale Mode=::NewEnumerator1
    Scale RGBA=&MakeVector4_FromLinearColor(LinearColor=&ColorFromCurve(ColorCurve=DataInterface() {propertyValues {
        AlphaCurve {keys [<at 0 0 RCIM_Cubic>,<at 0.1765 0.1>,<at 1 0 RCIM_Cubic>]}
        BlueCurve {keys [<at 0.247 1 RCIM_Cubic>]}
        GreenCurve {keys [<at 0.247 1 RCIM_Cubic>]}
        RedCurve {keys [<at 0.247 1 RCIM_Cubic>]}
       }}))
    ScaleRGBA=true
   )
   &ScaleColor(Scale Alpha=$User.AlphaMult, ScaleA=true, ScaleRGB=false) {module=ScaleColor001}
  ]
  rendererClasses [&NiagaraSpriteRendererProperties]
  renderers [{rendererClass=&NiagaraSpriteRendererProperties}]
 }
 Haze {
  emitterUpdateScript [
   &EmitterState()
   &SpawnRate(SpawnRate=11, Use Spawn Probability=false)
  ]
  particleSpawnScript [
   &InitializeParticle(
    Color=&RandomRangeLinearColor(Minimum=<rgba 0 0.2365 0.7344 1>, Maximum=<rgba 0.0223 0.7689 1 1>)
    Lifetime Max=3
    Lifetime Min=1.75
    Lifetime Mode=::NewEnumerator1
    Sprite Size Mode=::NewEnumerator3
    Uniform Sprite Size=&RandomRangeFloat(Minimum=175, Maximum=220)
    UsePositionOffset=false
    Write Material Random=true
   )
   &AddVelocity(Velocity=<0 0 15>)
   &ShapeLocation(
    Cylinder Height=130
    Cylinder Height Midpoint=0
    Enable Lathe Profile=true
    Lathe Profile=DataInterface() {propertyValues {Curve {keys [{tangentMode=RCTM_User, time=0, value=1}, {tangentMode=RCTM_User, time=1, value=0.5}]}}}
    MaskSpawnGroup=false
    Non Uniform Scale=<1 0.52 1>
    Offset=<0 -15 0>
    Offset Mode=::NewEnumerator0
    Override Seed=false
    Shape Primitive=::NewEnumerator1
    Surface Only=false
    UseOwnerScale=true
   )
   &CameraOffset(Camera Offset Amount=-100)
  ]
  particleUpdateScript [
   &ParticleState()
   &ScaleColor(
    Scale Alpha=&FloatFromCurve(
     FloatCurve=DataInterface() {propertyValues {Curve {keys [
         {interpMode=RCIM_Cubic, tangentMode=RCTM_User, time=0, value=0}
         {interpMode=RCIM_Cubic, tangentMode=RCTM_User, time=0.5, value=1}
         {tangentMode=RCTM_User, time=1, value=0}
        ]}}}
     Scale Curve=&Multiply_Float(A=0.2, B=$User.AlphaMult)
    )
    Scale RGB=&VectorFromFloat(Value=1)
    ScaleA=true
    ScaleRGB=true
   )
   &SolveForcesAndVelocity(Clamp Velocity=false, Limit Acceleration=false)
   &DynamicMaterialParameters(
    Index 0 Param 1=&RandomRangeFloat(Minimum=50, Maximum=100)
    Index 0 Param 2=&RandomRangeFloat(Minimum=0, Maximum=1)
    Index 0 Param 3=&RandomRangeFloat(Minimum=0, Maximum=1)
    Index 0 Param 4=&FloatFromCurve(FloatCurve=DataInterface() {propertyValues {Curve {keys [
         {interpMode=RCIM_Cubic, tangentMode=RCTM_User, time=0, value=1}
         {interpMode=RCIM_Cubic, tangentMode=RCTM_User, time=0.5, value=0.9}
         {tangentMode=RCTM_User, time=1, value=1}
        ]}}})
    Module.Index0Param0Write=true
    Module.Index0Param1Write=true
    Module.Index0Param2Write=true
    Module.Index0Param3Write=true
   )
  ]
  rendererClasses [&NiagaraSpriteRendererProperties]
  renderers [{rendererClass=&NiagaraSpriteRendererProperties}]
 }
 Smoke {
  emitterUpdateScript [
   &EmitterState()
   &SpawnRate(SpawnRate=10, Use Spawn Probability=false)
   &SpawnBurst_Instantaneous(Spawn Count=&RandomRangeInt(Minimum=4, Maximum=6), Use Spawn Probability=false) {enabled=false}
  ]
  particleSpawnScript [
   /SetVariables(Particles.SpriteRotationRate=&RandomRangeFloat(Minimum=-0.25, Maximum=0.25, Evaluation Type=::NewEnumerator1)) {isSetParameters=true, module=SetVariables_#1}
   &InitializeParticle(
    Color=&RandomRangeLinearColor(Minimum=<rgba 0.1 0.1 0.1 1>, Maximum=<rgba 0.7 0.7 0.7 1>)
    Lifetime=&RandomRangeFloat(Minimum=10, Maximum=12)
    Mass Mode=::NewEnumerator1
    Sprite Rotation Mode=::NewEnumerator2
    Sprite Size=&Vector2DFromFloat(Value=&RandomRangeFloat(Minimum=35, Maximum=40))
    Sprite Size Mode=::NewEnumerator4
    Sprite UV Mode=::NewEnumerator3
    UsePositionOffset=false
    Write Material Random=true
   ) {module=InitializeParticle001}
   &Color(Color=$User.Color_02) {enabled=false}
   /Niagara/Modules/Spawn/Location/SphereLocation(MaskSpawnGroup=false, Module.Override Randomness=false, Sphere Radius=40, Surface Only=true) {enabled=false}
   &DynamicMaterialParameters(
    Index 0 Param 1=&RandomRangeFloat(Minimum=0, Maximum=1, Evaluation Type=::NewEnumerator1)
    Index 0 Param 2=&RandomRangeFloat(Minimum=2, Maximum=4, Evaluation Type=::NewEnumerator1)
    Module.Param0WriteEnabled=true
    Module.Param1WriteEnabled=true
    Module.Param2WriteEnabled=false
    Module.Param3WriteEnabled=false
   )
  ]
  particleUpdateScript [
   &ParticleState()
   &ScaleColor(
    Scale Alpha=&FloatFromCurve(
     FloatCurve=DataInterface() {propertyValues {Curve {keys [
         <at 0 0 RCIM_Cubic>
         {arriveTangent=4.3234, interpMode=RCIM_Cubic, leaveTangent=4.3234, tangentMode=RCTM_User, time=0.1965, value=0.754}
         {arriveTangent=0.2786, interpMode=RCIM_Cubic, leaveTangent=0.2786, time=0.3, value=1}
         {arriveTangent=-0.1456, interpMode=RCIM_Cubic, leaveTangent=-0.1456, tangentMode=RCTM_User, time=0.9, value=0.95}
         <at 1 0 RCIM_Cubic>
        ]}}}
     Scale Curve=&Multiply_Float(A=1, B=$User.AlphaMult)
    )
    Scale RGB=&FN_Select_Vector_from_Time_Of_Day(Morning=<1 1 1>, Day=<1 1 1>, Evening=<1 1 1>, Night=<0.5 0.5 0.5>)
    ScaleA=true
    ScaleRGB=true
   )
   /Niagara/Modules/Update/SubUV/SubUVAnimation(Number Of Frames=62)
   &DynamicMaterialParameters(Module.Param0WriteEnabled=false, Module.Param1WriteEnabled=false, Module.Param2WriteEnabled=true, Module.Param3WriteEnabled=true) {module=DynamicMaterialParameters001}
   &ScaleSpriteSize(Scale Factor=&Vector2DFromFloat(Value=&FloatFromCurve(
      FloatCurve=DataInterface() {propertyValues {Curve {keys [{arriveTangent=0.7932, interpMode=RCIM_Cubic, time=0, value=0.5}, {interpMode=RCIM_Cubic, leaveTangent=0.3663, time=1.0038, value=1}]}}}
      Scale Curve=3
     )))
   &SkeletalMeshLocation(
    Bone Sampling Mode=::NewEnumerator4
    Inherit Velocity=false
    "Inherit Velocity (Scale)"=0.25
    MaskSpawnGroup=false
    Sampled Position Offset=&Add_Vector(
     A=&RandomRangeVector(Minimum=<-10 -10 -30>, Maximum=<10 10 30>)
     B=&MakeVector(
      X=&Multiply_Float(
       A=&RandomRangeFloat(Minimum=-1, Maximum=1)
       B=&FloatFromCurve(
        FloatCurve=DataInterface() {propertyValues {Curve {keys [<at 0 0 RCIM_Cubic>, {arriveTangent=2.1429, interpMode=RCIM_Cubic, leaveTangent=2.1429, time=0.5, value=0.325}, <at 0.7 1.5 RCIM_Cubic>]}}}
        Scale Curve=40
       )
      )
      Y=&Multiply_Float(
       A=&FloatFromCurve(
        FloatCurve=DataInterface() {propertyValues {Curve {keys [
            {arriveTangent=0.1375, interpMode=RCIM_Cubic, leaveTangent=0.1375, tangentMode=RCTM_User, time=0, value=0}
            {arriveTangent=0.3828, interpMode=RCIM_Cubic, leaveTangent=0.3828, tangentMode=RCTM_User, time=0.6, value=0.1}
            <at 0.8 1 RCIM_Cubic>
           ]}}}
        Scale Curve=30
       )
       B=1
      )
      Z=&FloatFromCurve(
       FloatCurve=DataInterface() {propertyValues {Curve {keys [{arriveTangent=3.0363, interpMode=RCIM_Cubic, leaveTangent=3.0363, tangentMode=RCTM_User, time=0, value=0}, <at 0.6 1.5 RCIM_Cubic>]}}}
       Scale Curve=-60
      )
     )
    )
    Sampled Position Offset Coordinate Space:ENiagaraMeshTransforms=::NewEnumerator2
    Skeletal Mesh=$User.SkeletalMesh
   ) {module=SkeletalMeshLocation002}
   &SpriteRotationRate(Rotation Rate=$Particles.Initial.SpriteRotationRate) {module=SpriteRotationRate001}
   &ScaleColor(Scale RGB=<1 1.5 2>, ScaleA=true, ScaleRGB=true) {enabled=false, module=ScaleColor001}
  ]
  rendererClasses [&NiagaraSpriteRendererProperties]
  renderers [{rendererClass=&NiagaraSpriteRendererProperties}]
 }
 SmokeLow {
  emitterUpdateScript [
   &EmitterState()
   &SpawnRate(SpawnRate=8, Use Spawn Probability=false)
   &SpawnBurst_Instantaneous(Spawn Count=&RandomRangeInt(Minimum=4, Maximum=6), Use Spawn Probability=false) {enabled=false}
  ]
  particleSpawnScript [
   /SetVariables(Particles.SpriteRotationRate=&RandomRangeFloat(Minimum=-0.025, Maximum=0.025, Evaluation Type=::NewEnumerator1)) {isSetParameters=true, module=SetVariables_#1}
   &InitializeParticle(
    Color=&RandomRangeLinearColor(Minimum=<rgba 0.1 0.1 0.1 1>, Maximum=<rgba 0.7 0.7 0.7 1>)
    Lifetime=&RandomRangeFloat(Minimum=2.5, Maximum=3.5)
    Mass Mode=::NewEnumerator1
    Sprite Rotation Mode=::NewEnumerator2
    Sprite Size=&Vector2DFromFloat(Value=&RandomRangeFloat(Minimum=35, Maximum=40))
    Sprite Size Mode=::NewEnumerator4
    Sprite UV Mode=::NewEnumerator3
    UsePositionOffset=false
    Write Material Random=true
   ) {module=InitializeParticle001}
   &Color(Color=$User.Color_02) {enabled=false}
   &SphereLocation(MaskSpawnGroup=false, Module.Override Randomness=false, Sphere Radius=40, Surface Only=true) {enabled=false}
   &DynamicMaterialParameters(
    Index 0 Param 1=&RandomRangeFloat(Minimum=0, Maximum=1, Evaluation Type=::NewEnumerator1)
    Index 0 Param 2=&RandomRangeFloat(Minimum=2, Maximum=4, Evaluation Type=::NewEnumerator1)
    Module.Param0WriteEnabled=true
    Module.Param1WriteEnabled=true
    Module.Param2WriteEnabled=false
    Module.Param3WriteEnabled=false
   )
   &SkeletalMeshLocation(
    Bone Sampling Mode=::NewEnumerator4
    Inherit Velocity=false
    "Inherit Velocity (Scale)"=0.25
    MaskSpawnGroup=false
    Sampled Position Offset=&Add_Vector(A=&RandomRangeVector(Minimum=<-30 -5 -100>, Maximum=<30 -5 -100>), B=<0 0 0>)
    Sampled Position Offset Coordinate Space:ENiagaraMeshTransforms=::NewEnumerator2
    Skeletal Mesh=$User.SkeletalMesh
   )
  ]
  particleUpdateScript [
   &ParticleState()
   &ScaleColor(
    Scale Alpha=&FloatFromCurve(
     FloatCurve=DataInterface() {propertyValues {Curve {keys [
         <at 0 0 RCIM_Cubic>
         {arriveTangent=-0.0566, interpMode=RCIM_Cubic, leaveTangent=-0.0566, tangentMode=RCTM_User, time=0.1, value=1}
         <at 1 0 RCIM_Cubic>
        ]}}}
     Scale Curve=&Multiply_Float(A=1, B=$User.AlphaMult)
    )
    Scale RGB=&FN_Select_Vector_from_Time_Of_Day(Morning=<1 1 1>, Day=<1 1 1>, Evening=<1 1 1>, Night=<0.5 0.5 0.5>)
    ScaleA=true
    ScaleRGB=true
   )
   &SubUVAnimation(Number Of Frames=32) {enabled=false}
   /Niagara/Modules/Update/SubUV/V2/SubUVAnimation(
    End Frame Range Override=49
    "Sprite Renderer "=DataInterface() {propertyValues {SpriteRenderer=&self:SmokeLow_2.NiagaraSpriteRendererProperties_0}}
    Start Frame Range Override=17
    UseEndFrame=true
    UseStartFrame=true
   ) {module=SubUVAnimation001}
   &DynamicMaterialParameters(Module.Param0WriteEnabled=false, Module.Param1WriteEnabled=false, Module.Param2WriteEnabled=true, Module.Param3WriteEnabled=true) {module=DynamicMaterialParameters001}
   &ScaleSpriteSize(Scale Factor=&Vector2DFromFloat(Value=&FloatFromCurve(
      FloatCurve=DataInterface() {propertyValues {Curve {keys [
          {arriveTangent=-0.5, interpMode=RCIM_Cubic, time=0, value=0}
          {arriveTangent=0.5, interpMode=RCIM_Cubic, leaveTangent=0.5, time=0.1, value=1}
          <at 0.5 0.25 RCIM_Cubic>
         ]}}}
      Scale Curve=3
     )))
   &SkeletalMeshLocation(
    Bone Sampling Mode=::NewEnumerator4
    Inherit Velocity=false
    "Inherit Velocity (Scale)"=0.25
    MaskSpawnGroup=false
    Sampled Position Offset=&Add_Vector(
     A=&RandomRangeVector(Minimum=<-30 -5 -110>, Maximum=<30 -5 -110>)
     B=&MakeVector(
      X=&Multiply_Float(
       A=&RandomRangeFloat(Minimum=-1, Maximum=1)
       B=&FloatFromCurve(
        FloatCurve=DataInterface() {propertyValues {Curve {keys [{arriveTangent=6.3743, interpMode=RCIM_Cubic, leaveTangent=6.3743, tangentMode=RCTM_User, time=0, value=0}, <at 0.5 1.5 RCIM_Cubic>]}}}
        Scale Curve=&RandomRangeFloat(Minimum=30, Maximum=60)
       )
      )
      Y=&Multiply_Float(
       A=&FloatFromCurve(
        FloatCurve=DataInterface() {propertyValues {Curve {keys [<at 0 0 RCIM_Cubic>,<at 0.8 1 RCIM_Cubic>]}}}
        Scale Curve=&RandomRangeFloat(Minimum=-40, Maximum=-120)
       )
       B=1
      )
      Z=&FloatFromCurve(
       FloatCurve=DataInterface() {propertyValues {Curve {keys [{arriveTangent=3.0363, interpMode=RCIM_Cubic, leaveTangent=3.0363, tangentMode=RCTM_User, time=0, value=0}, <at 0.6 1.5 RCIM_Cubic>]}}}
       Scale Curve=0
      )
     )
    )
    Sampled Position Offset Coordinate Space:ENiagaraMeshTransforms=::NewEnumerator2
    Skeletal Mesh=$User.SkeletalMesh
   ) {enabled=false, module=SkeletalMeshLocation002}
   &SpriteRotationRate(Rotation Rate=$Particles.Initial.SpriteRotationRate) {module=SpriteRotationRate001}
   &ScaleColor(Scale RGB=<1 1.5 2>, ScaleA=true, ScaleRGB=true) {enabled=false, module=ScaleColor001}
   &InheritSourceMovement(Applied Position Scale=<0.15 0.15 0.15>, ApplyForce=false, ApplyPosition=true, ApplyPositionFalloff=false)
   &AccelerationForce(Acceleration=<0 0 -10>)
   &SolveForcesAndVelocity(Clamp Velocity=false, Limit Acceleration=false)
  ]
  rendererClasses [&NiagaraSpriteRendererProperties]
  renderers [{rendererClass=&NiagaraSpriteRendererProperties}]
 }
}
summary {
 systemName=NS_Character_KnishClamp_Idle
 emitters [
  {emitterName=Haze, rendererClasses [&NiagaraSpriteRendererProperties]}
  {emitterName=Glow, rendererClasses [&NiagaraSpriteRendererProperties]}
  {emitterName=Smoke, rendererClasses [&NiagaraSpriteRendererProperties]}
  {emitterName=SmokeLow, rendererClasses [&NiagaraSpriteRendererProperties]}
 ]
 userVariables [
  User.AlphaMult {type=NiagaraFloat, defaultValue=1, description=""}
  User.SkeletalMesh {
   type=NiagaraDataInterfaceSkeletalMesh
   defaultValue {
    struct=NiagaraExt_VariableValue_DataInterface
    value {dataInterface=&self:NiagaraDataInterfaceSkeletalMesh_0}
   }
  }
 ]
}
