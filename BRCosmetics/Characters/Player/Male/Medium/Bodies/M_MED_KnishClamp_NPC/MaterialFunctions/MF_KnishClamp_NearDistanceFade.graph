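# Material function for the KnishClamp NPC that dissolves its opacity mask when the camera is too near or too far.
# Object-to-camera distance builds a band mask from start and max fade parameters, also exposed as a Distance Mask output.
# An inlined temporal dither, using screen pixels, sample index and scrolling noise, turns that band into a stippled fade.
# A custom primitive data parameter blends the effect over the incoming material attributes.
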MaterialFunction /BRCosmetics/Characters/Player/Male/Medium/Bodies/M_MED_KnishClamp_NPC/MaterialFunctions/MF_KnishClamp_NearDistanceFade
graph {nodes [
  n1 = FunctionInput() {inputType=FunctionInput_MaterialAttributes}
  n2 = GetMaterialAttributes(None=@n1) {attributeGetTypes [#1]}
  n3 = Distance(A=ObjectPositionWS(), B=CameraPositionWS())
  n4 = Subtract(A=@n3, B=ScalarParameter() {defaultValue=500, group=Knish | Dissolve, parameterName=Knish - Start Fade Distance, sortPriority=44})
  n5 = ScalarParameter() {defaultValue=300, group=Knish | Dissolve, parameterName=Knish - Start Fade Length, sortPriority=44}
  n6 = Divide(
   A=Subtract(A=@n3, B=ScalarParameter() {defaultValue=5000, group=Knish | Dissolve, parameterName=Knish - Max Fade Distance, sortPriority=45})
   B=ScalarParameter() {defaultValue=300, group=Knish | Dissolve, parameterName=Knish - Max Fade Length, sortPriority=45}
  )
  n7 = Multiply(A=Saturate(None=Divide(A=@n4, B=@n5)), B=OneMinus(None=Saturate(None=@n6)))
  n8 = ViewProperty() {property=MEVP_TemporalSampleIndex}
  MaterialExpressionAdd_2 = Add(A=ScreenPosition().PixelPosition, B=AppendVector(A=@n8.Property, B=@n8.Property))
  n9 = Custom!(p=@MaterialExpressionAdd_2) {
   code="Mod( ((uint)(p.x) + 2 * (uint)(p.y)) , 5 )"
   outputType=CMOT_Float1
   inputs [{
     inputName=p
     input {expression=&self:MaterialExpressionAdd_2}
    }]
  }
  n10 = Add(
   A=MaterialFunctionCall!(TextureResolution=Constant2Vector() {g=256, r=256}) {materialFunction=/Engine/Functions/Engine_MaterialFunctions02/Texturing/ScreenAlignedPixelToPixelUVs}.Result
   B=Multiply(A=Time()) {constB=0.3}
  )
  n11 = Multiply(A=Add(A=@n9, B=TextureSample(UVs=@n10) {samplerType=SAMPLERTYPE_LinearGrayscale, texture=/Engine/EngineMaterials/Good64x64TilingNoiseHighFreq}.R) {constB=0}) {constB=0.1666, desc=Divide by 6 avoiding a precision problem where an Alpha Threshold of 0 still leaves some remaining pixels.}
  n12 = LinearInterpolate(B=@n2.OpacityMask, Alpha=Add(A=Add(A=@n7, B=@n11) {constB=0.5}) {constB=-0.5})
  FunctionOutput(None=BlendMaterialAttributes(
    A=@n1
    B=SetMaterialAttributes!(MaterialAttributes=@n2.MaterialAttributes, OpacityMask=@n12) {attributeSetTypes [#1]}
    Alpha=ScalarParameter() {bUseCustomPrimitiveData=true, group=Knish | Dissolve, parameterName=CustomPrim Use Distance FX, sortPriority=40}
   ))
  FunctionOutput(None=@n7) {outputName=Distance Mask, sortPriority=2}
 ]}
