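# Additive unlit parent material for the KnishClamp NPC's mirage effect, a shimmering, refractive glow with Niagara sprite support.
# Distorted, panning noise, radial masks, depth fade and camera-distance fades build opacity, with particle color, dynamic parameters or time-of-day color feeding the graph's variants.
# Two panning streak samples drive refraction.
# World position offset builds a camera-facing quad scaled by distance, with distance effects toggled through custom primitive data.
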Material /BRCosmetics/Characters/Player/Male/Medium/Bodies/M_MED_KnishClamp_NPC/Materials/M_KnishClamp_Mirage_Parent
graph {
 properties {
  bAutomaticallySetUsageInEditor=false
  bCastRayTracedShadows=false
  bUsedWithMorphTargets=true
  bUsedWithNiagaraSprites=true
  blendMode=BLEND_Additive
  pixelDepthOffsetMode=PDOM_Legacy
  refractionMethod=RM_IndexOfRefraction
  shadingModel=MSM_Unlit
  usedShadingModels=Unlit
 }
 nodes [
  n1 = Distance(A=ObjectPositionWS(), B=CameraPositionWS())
  n2 = Subtract(A=@n1, B=ScalarParameter() {defaultValue=2500, group=Knish | Dissolve, parameterName=Knish - Start Fade Distance, sortPriority=44})
  n3 = ScalarParameter() {defaultValue=600, group=Knish | Dissolve, parameterName=Knish - Start Fade Length, sortPriority=44}
  n4 = Divide(
   A=Subtract(A=@n1, B=ScalarParameter() {defaultValue=10000, group=Knish | Dissolve, parameterName=Knish - Max Fade Distance, sortPriority=45})
   B=ScalarParameter() {defaultValue=2000, group=Knish | Dissolve, parameterName=Knish - Max Fade Length, sortPriority=45}
  )
  n5 = Multiply(A=Saturate(None=Divide(A=@n2, B=@n3)), B=OneMinus(None=Saturate(None=@n4)))
  n6 = ViewProperty() {property=MEVP_TemporalSampleIndex}
  MaterialExpressionAdd_22 = Add(A=ScreenPosition().PixelPosition, B=AppendVector(A=@n6.Property, B=@n6.Property))
  n7 = Custom!(p=@MaterialExpressionAdd_22) {
   code="Mod( ((uint)(p.x) + 2 * (uint)(p.y)) , 5 )"
   outputType=CMOT_Float1
   inputs [{
     inputName=p
     input {expression=&self:MaterialExpressionAdd_22}
    }]
  }
  n8 = MaterialFunctionCall!(TextureResolution=Constant2Vector() {g=256, r=256}) {materialFunction=/Engine/Functions/Engine_MaterialFunctions02/Texturing/ScreenAlignedPixelToPixelUVs}
  n9 = Multiply(A=Time()) {constB=0.3}
  n10 = Add(
   A=@n7
   B=TextureSample(UVs=Add(A=@n8.Result, B=@n9)) {samplerType=SAMPLERTYPE_LinearGrayscale, texture=/Engine/EngineMaterials/Good64x64TilingNoiseHighFreq}.R
  ) {constB=0}
  Add(A=Add(
    A=@n5
    B=Multiply(A=@n10) {constB=0.1666, desc=Divide by 6 avoiding a precision problem where an Alpha Threshold of 0 still leaves some remaining pixels.}
   ) {constB=0.5}) {constB=-0.5}
  Constant() {r=1}
  Constant() {r=1}
  LinearInterpolate(
   B=NamedRerouteUsage()
   Alpha=LinearInterpolate(A=Constant() {r=1}, B=NamedRerouteUsage(), Alpha=ScalarParameter() {defaultValue=1, parameterName=UseDistanceFade})
  ) {constA=1}
  n11 = TextureCoordinate() {uTiling=0.25, vTiling=0.25}
  MaterialFunctionCall(Vector or UVs=Add(A=@n11) {constB=-0.125}) {materialFunction=&VectorToRadialValue}
  MaterialFunctionCall!(In) {materialFunction=/BRCosmetics/Characters/Player/Male/Medium/Bodies/M_MED_KnishClamp_NPC/MaterialFunctions/MF_KnishClamp_NearDistanceFade}
  n12 = TextureCoordinate()
  Multiply(
   A=@n12
   B=MaterialFunctionCall(X=Constant() {r=1}, Y=Constant() {r=0.5}) {materialFunction=&MakeFloat2}.Result
  )
  n13 = ScalarParameter() {defaultValue=1, group=Color, parameterName=Emissive Intensity Day, sortPriority=6}
  n14 = ScalarParameter() {defaultValue=1, group=Color, parameterName=Emissive Intensity Night, sortPriority=7}
  n15 = LinearInterpolate(A=@n13, B=@n14)
  Multiply(
   A=VectorParameter() {group=Color, parameterName=Color, sortPriority=5, defaultValue=<rgba 0 0.7566 1 1>}.RGB
   B=MaterialFunctionCall(Value_Morning=@n15, Value_Day=@n13, Value_Evening=@n15, Value_Night=@n14) {materialFunction=&Scalar_Time_of_Day}.Result
  )
  n16 = Distance(A=ObjectPositionWS(), B=CameraPositionWS())
  n17 = Divide(
   A=Subtract(A=@n16, B=ScalarParameter() {defaultValue=500, group=Fade, parameterName=Max Distance, sortPriority=50})
   B=ScalarParameter() {defaultValue=300, group=Fade, parameterName=Max Distance Fade Length, sortPriority=50}
  )
  n18 = Subtract(A=@n16, B=ScalarParameter() {defaultValue=500, group=Fade, parameterName=Fade Start Distance, sortPriority=50})
  n19 = ScalarParameter() {defaultValue=300, group=Fade, parameterName=Fade Length, sortPriority=50}
  n20 = Multiply(A=OneMinus(None=Saturate(None=@n17)), B=Saturate(None=Divide(A=@n18, B=@n19)))
  n21 = ScalarParameter() {bUseCustomPrimitiveData=true, group=Knish | Dissolve, parameterName=CustomPrim Use Distance FX, sortPriority=40}
  n22 = Constant() {r=1}
  n23 = DepthFade(
   Opacity=MaterialFunctionCall(Distance=ScalarParameter() {group=Fade, parameterName=Near Camera Fade}) {materialFunction=&Near_Camera_Fade}.Result
   FadeDistance=DynamicParameter() {paramNames [Depth Fade, RandX, RandY, Refraction]}.Depth Fade
  )
  n24 = LinearInterpolate(
   A=ScalarParameter() {defaultValue=1, group=Fade, parameterName=Airglow Opacity, sortPriority=6}
   B=Multiply(
    A=ScalarParameter() {defaultValue=1, group=Fade, parameterName=Airglow Opacity, sortPriority=6}
    B=ShadingPathSwitch(Default=@n22, Deferred=@n23)
   )
   Alpha=ScalarParameter() {defaultValue=1, group=Fade, parameterName=Depth Fade Blend Intensity}
  )
  n25 = MaterialFunctionCall(
   Radius=ScalarParameter() {defaultValue=0.45, group=Mask, parameterName=Radial Gradient Radius, sortPriority=40}
   Density=ScalarParameter() {defaultValue=2.33, group=Mask, parameterName=Radial Gradient Density, sortPriority=40}
  ) {materialFunction=&RadialGradientExponential}
  n26 = Constant() {r=2}
  n27 = Constant() {r=0.5}
  n28 = Multiply(
   A=Multiply(A=@n11, B=AppendVector(A=@n26, B=@n27))
   B=AppendVector(
    A=ScalarParameter() {defaultValue=2, group=Silhouette, parameterName=U Pan Tiling}
    B=ScalarParameter() {defaultValue=0.5, group=Silhouette, parameterName=V Pan Tiling}
   )
  )
  n29 = Multiply(A=Time(), B=Constant() {r=0.7})
  n30 = AppendVector(
   A=ScalarParameter() {defaultValue=0.005, group=Silhouette, parameterName=U Pan Speed}
   B=ScalarParameter() {defaultValue=0.01, group=Silhouette, parameterName=V Pan Speed}
  )
  n31 = TextureSample(UVs=Add(A=Panner(Coordinate=@n28, Time=@n29, Speed=@n30) {speedY=0.121}, B=NamedRerouteUsage())) {samplerType=SAMPLERTYPE_Normal, texture=/Game/Gadgets/Assets/VinderTech_GliderChute/Glider_Vampire/Textures/T_Glider_Bat_Noises_N}
  n32 = TextureCoordinate() {uTiling=0.25, vTiling=0.25}
  n33 = AppendVector(A=Constant() {r=2}, B=Constant() {r=1})
  n34 = ScalarParameter() {defaultValue=2, group=Silhouette, parameterName=U Pan Tiling}
  n35 = ScalarParameter() {defaultValue=0.5, group=Silhouette, parameterName=V Pan Tiling}
  n36 = Panner(
   Coordinate=Multiply(A=Multiply(A=@n32, B=@n33), B=AppendVector(A=@n34, B=@n35))
   Time=@n29
   Speed=AppendVector(
    A=ScalarParameter() {defaultValue=0.005, group=Silhouette, parameterName=U Pan Speed}
    B=ScalarParameter() {defaultValue=0.01, group=Silhouette, parameterName=V Pan Speed}
   )
  ) {speedY=0.121}
  n37 = NamedRerouteUsage()
  n38 = Multiply(A=TextureSample(UVs=Add(A=@n36, B=@n37)) {samplerType=SAMPLERTYPE_Normal, texture=/Game/VisualThreatManager/StormVisuals/Textures/TilingNoise06_N}.RGB) {constB=0.35}
  n39 = Multiply(A=Add(A=ComponentMask(None=@n31.RGB) {g=true, r=true}, B=ComponentMask(None=@n38) {g=true, r=true})) {constB=0.1}
  n40 = MaterialFunctionCall(X=ScalarParameter() {defaultValue=0.05, parameterName=X}, Y=ScalarParameter() {defaultValue=0.1, parameterName=Y}) {materialFunction=&MakeFloat2}
  n41 = Add(A=Add(A=Multiply(A=@n39) {constB=2}, B=Panner(Coordinate=@n12, Time=@n29, Speed=@n40.Result)), B=NamedRerouteUsage())
  n42 = TextureSample(UVs=Add(A=@n39, B=TextureCoordinate())) {texture=/Game/Packages/Fortress_Transylvania/Haunted_Forest/Swamps/Textures/T_dust_cloud_01}
  n43 = Multiply(
   A=Multiply(A=@n42.R, B=@n42.R)
   B=LinearInterpolate(Alpha=TextureSample(UVs=@n41) {samplerType=SAMPLERTYPE_Masks, texture=/Game/Weapons/WeaponSkins/Wraps/Textures/Mash/Mash_Noise}.R)
  )
  n44 = Multiply(A=Multiply(A=LinearInterpolate(B=@n20, Alpha=@n21) {constA=1}, B=Multiply(A=@n24, B=@n25.RadialGradientExponential)), B=@n43)
  NamedRerouteDeclaration(None=@n44) {name=Opacity, nodeColor=<rgba 0.0078 0 1 1>}
  n45 = Multiply(
   A=TextureCoordinate()
   B=AppendVector(
    A=ScalarParameter() {defaultValue=2, group=Silhouette, parameterName=U Pan Tiling}
    B=ScalarParameter() {defaultValue=0.5, group=Silhouette, parameterName=V Pan Tiling}
   )
  )
  n46 = AppendVector(
   A=ScalarParameter() {defaultValue=0.005, group=Silhouette, parameterName=U Pan Speed}
   B=ScalarParameter() {defaultValue=0.01, group=Silhouette, parameterName=V Pan Speed}
  )
  n47 = TextureSample(UVs=Add(A=Panner(Coordinate=@n45, Time=@n29, Speed=@n46) {speedY=0.121}, B=NamedRerouteUsage())) {samplerType=SAMPLERTYPE_Normal, texture=/Game/Gadgets/Assets/Telephone_Booth/Textures/T_BlindsNormalNoise_N}
  NamedRerouteDeclaration(None=@n47.RGB) {name=Normal, nodeColor=<rgba 0.3125 1 0 1>}
  n48 = NamedRerouteUsage()
  n49 = MaterialFunctionCall(
   Radius=ScalarParameter() {defaultValue=0.48, group=Refraction, parameterName=Refraction Falloff Radius, sortPriority=40}
   Density=ScalarParameter() {defaultValue=2.33, group=Refraction, parameterName=Refraction Falloff Density, sortPriority=40}
  ) {materialFunction=&RadialGradientExponential}
  NamedRerouteDeclaration(None=LinearInterpolate(
    A=Constant() {r=1}
    B=DynamicParameter() {paramNames [Depth Fade, RandX, RandY, Refraction]}.Refraction
    Alpha=Multiply(A=@n48, B=@n49.RadialGradientExponential)
   )) {name=Refraction, nodeColor=<rgba 1 0 0.6328 1>}
  NamedRerouteDeclaration(None=@n43) {name=ForRefr, nodeColor=<rgba 0 1 0.0859 1>}
  n50 = DynamicParameter() {paramNames [Depth Fade, RandX, RandY, Refraction]}
  NamedRerouteDeclaration(None=MaterialFunctionCall(X=@n50.RandX, Y=@n50.RandX) {materialFunction=&MakeFloat2}.Result) {name=particlerandomuv, nodeColor=<rgba 1 0 0.9609 1>}
  n51 = ParticleColor()
  NamedRerouteDeclaration(None=@n51.RGB) {name=Emissive, nodeColor=<rgba 1 0.1875 0 1>}
  n52 = NamedRerouteUsage()
  n53 = Constant() {r=1}
  n54 = NamedRerouteUsage()
  n55 = ScalarParameter() {defaultValue=1, parameterName=UseDistanceFade}
  NamedRerouteDeclaration(None=Multiply(A=Multiply(A=@n51.A, B=@n52), B=LinearInterpolate(A=@n53, B=@n54, Alpha=@n55))) {name=FinalOpacity, nodeColor=<rgba 0.2656 0 1 1>}
  NamedRerouteDeclaration(None=@n5) {name=Distance, nodeColor=<rgba 0 0.8594 1 1>}
  NamedRerouteUsage()
  NamedRerouteUsage()
  NamedRerouteUsage()
  n56 = Distance(A=ObjectPositionWS(), B=CameraPositionWS())
  n57 = Divide(
   A=Subtract(A=@n56, B=ScalarParameter() {defaultValue=500, group=Fade, parameterName=Max Distance, sortPriority=50})
   B=ScalarParameter() {defaultValue=300, group=Fade, parameterName=Max Distance Fade Length, sortPriority=50}
  )
  n58 = Subtract(A=@n56, B=ScalarParameter() {defaultValue=500, group=Fade, parameterName=Fade Start Distance, sortPriority=50})
  n59 = ScalarParameter() {defaultValue=300, group=Fade, parameterName=Fade Length, sortPriority=50}
  n60 = Multiply(A=OneMinus(None=Saturate(None=@n57)), B=Saturate(None=Divide(A=@n58, B=@n59)))
  n61 = ScalarParameter() {bUseCustomPrimitiveData=true, group=Knish | Dissolve, parameterName=CustomPrim Use Distance FX, sortPriority=40}
  n62 = Constant() {r=1}
  n63 = DepthFade(
   Opacity=MaterialFunctionCall(Distance=ScalarParameter() {group=Fade, parameterName=Near Camera Fade}) {materialFunction=&Near_Camera_Fade}.Result
   FadeDistance=ScalarParameter() {defaultValue=100, group=Fade, parameterName=Depth Fade Distance}
  )
  n64 = LinearInterpolate(
   A=ScalarParameter() {defaultValue=1, group=Fade, parameterName=Airglow Opacity, sortPriority=6}
   B=Multiply(
    A=ScalarParameter() {defaultValue=1, group=Fade, parameterName=Airglow Opacity, sortPriority=6}
    B=ShadingPathSwitch(Default=@n62, Deferred=@n63)
   )
   Alpha=ScalarParameter() {defaultValue=1, group=Fade, parameterName=Depth Fade Blend Intensity}
  )
  n65 = ScalarParameter() {defaultValue=0.02, group=Mask, parameterName=Edge Mask Speed, sortPriority=40}
  n66 = Add(A=TextureCoordinate(), B=Multiply(A=Time(), B=@n65) {constB=0.2})
  n67 = TextureObject() {samplerType=SAMPLERTYPE_LinearGrayscale, texture=/Game/Global/Textures/Default/Flowmaps/T_StormSwirlMask}
  n68 = Multiply(A=Time(), B=@n65) {constB=0.2}
  n69 = Add(
   A=TextureCoordinate()
   B=MaterialFunctionCall(X=Multiply(A=@n68) {constB=-1}, Y=@n68) {materialFunction=&MakeFloat2}.Result
  )
  n70 = MaterialFunctionCall(
   X=Subtract(A=TextureSample(UVs=@n66, Tex=@n67) {samplerType=SAMPLERTYPE_LinearGrayscale, texture=&T_StormSwirlMask}.R) {constB=0.5}
   Y=Subtract(A=TextureSample(UVs=@n69, Tex=@n67) {samplerType=SAMPLERTYPE_LinearGrayscale, texture=&T_StormSwirlMask}.R) {constB=0.5}
  ) {materialFunction=&MakeFloat2}
  n71 = ScalarParameter() {defaultValue=0.1, group=Mask, parameterName=Edge Warp, sortPriority=40}
  n72 = MaterialFunctionCall(
   UVs=Add(A=TextureCoordinate(), B=Multiply(A=@n70.Result, B=@n71))
   Radius=ScalarParameter() {defaultValue=0.45, group=Mask, parameterName=Radial Gradient Radius, sortPriority=40}
   Density=ScalarParameter() {defaultValue=2.33, group=Mask, parameterName=Radial Gradient Density, sortPriority=40}
  ) {materialFunction=&RadialGradientExponential}
  Reroute(None=Multiply(A=LinearInterpolate(B=@n60, Alpha=@n61) {constA=1}, B=Multiply(A=@n64, B=@n72.RadialGradientExponential)))
  n73 = ScalarParameter() {group=WPO, parameterName=Camera Offset}
  n74 = ConstantBiasScale(None=TextureCoordinate()) {bias=-0.5, scale=1}
  n75 = ComponentMask(None=@n74) {r=true}
  n76 = ObjectPositionWS()
  n77 = CameraPositionWS()
  n78 = MaterialFunctionCall!(Vector1=Normalize(VectorInput=Subtract(A=@n76, B=@n77)), Vector2=Constant3Vector() {constant=<rgba 0 0 1 0>}) {materialFunction=/Engine/Functions/Engine_MaterialFunctions02/Math/CreateThirdOrthogonalVector}
  n79 = Distance(A=ObjectPositionWS(), B=CameraPositionWS())
  n80 = Subtract(A=@n79, B=ScalarParameter() {defaultValue=300, group=WPO, parameterName=WPO Max Distance Length, sortPriority=50})
  n81 = ScalarParameter() {defaultValue=300, group=WPO, parameterName=WPO Length, sortPriority=50}
  n82 = OneMinus(None=Saturate(None=Divide(A=@n80, B=@n81) {constB=50}))
  n83 = ScalarParameter() {defaultValue=500, group=WPO, parameterName=WPO Start Distance, sortPriority=50}
  n84 = Saturate(None=Divide(A=Subtract(A=@n79, B=@n83), B=ScalarParameter() {defaultValue=300, group=WPO, parameterName=WPO Length, sortPriority=50}) {constB=50})
  n85 = Multiply(
   A=ScalarParameter() {defaultValue=1, group=WPO, parameterName=Scale}
   B=LinearInterpolate(
    B=Multiply(A=@n82, B=@n84)
    Alpha=ScalarParameter() {bUseCustomPrimitiveData=true, group=Knish | Dissolve, parameterName=CustomPrim Use Distance FX, sortPriority=40}
   ) {constA=1}
  )
  n86 = Multiply(A=Normalize(VectorInput=@n78.Vector3), B=@n85)
  n87 = Multiply(A=@n78.Vector2, B=ComponentMask(None=@n74) {g=true}) {constA=1}
  n88 = Subtract(A=Add(A=Multiply(A=@n75, B=@n86), B=Multiply(A=@n87, B=@n85)), B=WorldPosition().XYZ)
  n89 = ObjectPositionWS()
  Reroute(None=Add(
    A=MaterialFunctionCall(Offset Amount=@n73) {materialFunction=/Engine/Functions/Engine_MaterialFunctions02/WorldPositionOffset/CameraOffset}.Camera Offset
    B=Add(A=@n88, B=@n89)
   ))
  ScalarParameter() {defaultValue=100, group=Fade, parameterName=Depth Fade Distance}
  n90 = ParticleColor()
  n91 = GetMaterialAttributes(None=SetMaterialAttributes!(
    MaterialAttributes=SetMaterialAttributes!(MaterialAttributes, Opacity=@n44, Normal=@n47.RGB, World Position Offset) {attributeSetTypes [#1, #2, #3]}
    Base Color=@n90.RGB
    Emissive Color=@n90.RGB
   ) {attributeSetTypes [#4, #5]}) {attributeGetTypes [#1]}
  n92 = TextureCoordinate()
  n93 = Multiply(A=@n92, B=ScalarParameter() {defaultValue=2, group=Refraction, parameterName=Refraction Mirage Tiling A, sortPriority=50})
  n94 = Time()
  n95 = VectorParameter() {group=Refraction, parameterName=Speed, sortPriority=51, defaultValue=<rgba 0.05 0.03 -0.025 0.04>}
  n96 = Multiply(A=@n94, B=MaterialFunctionCall(X=@n95.R, Y=@n95.R) {materialFunction=&MakeFloat2}.Result)
  n97 = TextureObject() {texture=/Game/Effects/Fort_Effects/Textures/Energy/T_OutpostShockwaveStreaks_02}
  n98 = Multiply(A=@n92, B=ScalarParameter() {defaultValue=1.33, group=Refraction, parameterName=Refraction Mirage Tiling B, sortPriority=50})
  n99 = Multiply(A=@n94, B=MaterialFunctionCall(X=@n95.B, Y=@n95.B) {materialFunction=&MakeFloat2}.Result)
  n100 = Add(
   A=TextureSample(UVs=Add(A=@n93, B=@n96), Tex=@n97) {texture=&T_OutpostShockwaveStreaks_02}.R
   B=TextureSample(UVs=Add(A=@n98, B=@n99), Tex=@n97) {texture=&T_OutpostShockwaveStreaks_02}.R
  )
  n101 = LinearInterpolate(B=ScalarParameter() {defaultValue=1.5, group=Refraction, parameterName=Refraction, sortPriority=30}, Alpha=Saturate(None=Divide(A=@n100))) {constA=1}
  n102 = Distance(A=CameraPositionWS(), B=ObjectPositionWS())
  n103 = ScalarParameter() {group=Refraction, parameterName=Refraction Start Fade Offset, sortPriority=33}
  n104 = Saturate(None=Divide(
    A=Subtract(A=@n102, B=@n103)
    B=ScalarParameter() {defaultValue=5000, group=Refraction, parameterName=Refraction Fade Out Length, sortPriority=34}
   ))
  n105 = MaterialFunctionCall(
   Radius=ScalarParameter() {defaultValue=0.48, group=Refraction, parameterName=Refraction Falloff Radius, sortPriority=40}
   Density=ScalarParameter() {defaultValue=2.33, group=Refraction, parameterName=Refraction Falloff Density, sortPriority=40}
  ) {materialFunction=&RadialGradientExponential}
  SetMaterialAttributes!(
   MaterialAttributes=@n91.MaterialAttributes
   Refraction=LinearInterpolate(B=LinearInterpolate(A=@n101, Alpha=@n104) {constA=1}, Alpha=LinearInterpolate(B=@n91.Opacity, Alpha=@n105.RadialGradientExponential)) {constA=1}
  ) {attributeSetTypes [#6]}
  n106 = ConstantBiasScale(None=TextureCoordinate()) {bias=-0.5, scale=1}
  n107 = ObjectPositionWS()
  n108 = CameraPositionWS()
  n109 = MaterialFunctionCall(Vector1=Normalize(VectorInput=Subtract(A=@n107, B=@n108)), Vector2=Constant3Vector() {constant=<rgba 0 0 1 0>}) {materialFunction=&CreateThirdOrthogonalVector}
  n110 = Normalize(VectorInput=@n109.Vector3)
  n111 = Distance(A=ObjectPositionWS(), B=CameraPositionWS())
  n112 = Subtract(A=@n111, B=ScalarParameter() {defaultValue=300, group=WPO, parameterName=WPO Max Distance Length, sortPriority=50})
  n113 = ScalarParameter() {defaultValue=300, group=WPO, parameterName=WPO Length, sortPriority=50}
  n114 = OneMinus(None=Saturate(None=Divide(A=@n112, B=@n113) {constB=50}))
  n115 = ScalarParameter() {defaultValue=500, group=WPO, parameterName=WPO Start Distance, sortPriority=50}
  n116 = Saturate(None=Divide(A=Subtract(A=@n111, B=@n115), B=ScalarParameter() {defaultValue=300, group=WPO, parameterName=WPO Length, sortPriority=50}) {constB=50})
  n117 = Multiply(
   A=ScalarParameter() {defaultValue=1, group=WPO, parameterName=Scale}
   B=LinearInterpolate(
    B=Multiply(A=@n114, B=@n116)
    Alpha=ScalarParameter() {bUseCustomPrimitiveData=true, group=Knish | Dissolve, parameterName=CustomPrim Use Distance FX, sortPriority=40}
   ) {constA=1}
  )
  n118 = ComponentMask(None=@n106) {g=true}
  n119 = Add(
   A=Multiply(A=ComponentMask(None=@n106) {r=true}, B=Multiply(A=@n110, B=@n117))
   B=Multiply(A=Multiply(A=@n109.Vector2, B=@n118) {constA=1}, B=@n117)
  )
  n120 = WorldPosition()
  n121 = Add(
   A=MaterialFunctionCall(Offset Amount=ScalarParameter() {group=WPO, parameterName=Camera Offset}) {materialFunction=&CameraOffset}.Camera Offset
   B=Add(A=Subtract(A=@n119, B=@n120.XYZ), B=ObjectPositionWS())
  )
  n122 = VectorParameter() {group=Color, parameterName=Color, sortPriority=5, defaultValue=<rgba 0 0.7566 1 1>}
  n123 = ScalarParameter() {defaultValue=1, group=Color, parameterName=Emissive Intensity Day, sortPriority=6}
  n124 = ScalarParameter() {defaultValue=1, group=Color, parameterName=Emissive Intensity Night, sortPriority=7}
  n125 = LinearInterpolate(A=@n123, B=@n124)
  n126 = MaterialFunctionCall(Value_Morning=@n125, Value_Day=@n123, Value_Evening=@n125, Value_Night=@n124) {materialFunction=&Scalar_Time_of_Day}
  n127 = GetMaterialAttributes(None=SetMaterialAttributes!(
    MaterialAttributes=SetMaterialAttributes!(MaterialAttributes, World Position Offset=@n121) {attributeSetTypes [#3]}
    Base Color=@n122.RGB
    Emissive Color=Multiply(A=@n122.RGB, B=@n126.Result)
   ) {attributeSetTypes [#4, #5]}) {attributeGetTypes [#1]}
  n128 = ScalarParameter() {defaultValue=1, parameterName=Refraction Min}
  n129 = ScalarParameter() {defaultValue=1, parameterName=Refraction Max}
  n130 = TextureCoordinate()
  n131 = ScalarParameter() {defaultValue=2, group=Refraction, parameterName=Tiling A}
  n132 = Time()
  n133 = VectorParameter() {parameterName=Mirage Pan Speed, channelNames {a=Y B, b=X B, g=Y A, r=X A}, defaultValue=<rgba 0.1 0.05 -0.066 0.07>}
  n134 = MaterialFunctionCall(X=@n133.X A, Y=@n133.X A) {materialFunction=&MakeFloat2}
  n135 = TextureObject() {texture=&T_OutpostShockwaveStreaks_02}
  n136 = TextureSample(UVs=Add(A=Multiply(A=@n130, B=@n131), B=Multiply(A=@n132, B=@n134.Result)), Tex=@n135) {texture=&T_OutpostShockwaveStreaks_02}
  n137 = ScalarParameter() {defaultValue=1.33, group=Refraction, parameterName=Tiling B}
  n138 = MaterialFunctionCall(X=@n133.X B, Y=@n133.X B) {materialFunction=&MakeFloat2}
  n139 = TextureSample(UVs=Add(A=Multiply(A=@n130, B=@n137), B=Multiply(A=@n132, B=@n138.Result)), Tex=@n135) {texture=&T_OutpostShockwaveStreaks_02}
  n140 = Saturate(None=Divide(A=Add(A=@n136.R, B=@n139.R)))
  SetMaterialAttributes!(
   MaterialAttributes=@n127.MaterialAttributes
   Refraction=LinearInterpolate(B=LinearInterpolate(A=@n128, B=@n129, Alpha=@n140), Alpha=@n127.Opacity) {constA=1}
  ) {attributeSetTypes [#6]}
 ]
}
