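# Material function for the KnishClamp NPC body that adds animated noise distortion to the incoming World Position Offset.
# A world-aligned Perlin texture scrolls over time, is centered around zero and becomes a vector with separately scaled Z.
# A pre-skinned local Z gradient, a strength multiplier input and an optional camera-distance fade mask that vector.
# A custom-primitive "Use Distance FX" lerp branches off the result but does not feed the output.
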MaterialFunction /BRCosmetics/Characters/Player/Male/Medium/Bodies/M_MED_KnishClamp_NPC/MaterialFunctions/MF_KnishClamp_Distortion
graph {nodes [
  n1 = GetMaterialAttributes(None=FunctionInput() {inputType=FunctionInput_MaterialAttributes}) {attributeGetTypes [#1]}
  n2 = TextureObject() {samplerType=SAMPLERTYPE_Masks, texture=/BRCosmetics/Characters/Player/Male/Large/Bodies/M_LRG_Clash_V/Textures/Simple_Perlin_M}
  n3 = ScalarParameter() {defaultValue=512, group=Knish | WPO, parameterName=Knish - WPO Distortion Tiling Size, sortPriority=55}
  n4 = Add(
   A=WorldPosition().XYZ
   B=Multiply(A=Time(), B=ScalarParameter() {defaultValue=0.3, group=Knish | WPO, parameterName=Knish - WPO Distortion Speed, sortPriority=55})
  )
  n5 = Subtract(A=ComponentMask(None=MaterialFunctionCall(TextureObject=@n2, TextureSize=@n3, WorldPosition=@n4) {materialFunction=&WorldAlignedTexture}.XY Texture) {r=true}) {constB=0.5}
  n6 = Multiply(A=@n5, B=ScalarParameter() {defaultValue=0.2, group=Knish | WPO, parameterName=Knish - WPO Distortion Z Multi, sortPriority=55}) {constB=0.2}
  n7 = Subtract(
   A=Divide(
    A=FunctionInput() {inputName=PreskinnedLocalPos, sortPriority=2}
    B=ScalarParameter() {defaultValue=130, group=Knish, parameterName=Knish - WPO Gradient Length}
   ) {constB=100}
   B=ScalarParameter() {group=Knish, parameterName=Knish - WPO Gradient Offset}
  )
  n8 = OneMinus(None=Saturate(None=ComponentMask(None=@n7) {b=true}))
  n9 = Distance(A=ObjectPositionWS(), B=CameraPositionWS())
  n10 = ScalarParameter() {defaultValue=500, group=Knish | WPO, parameterName=Knish - WPO Start Distance, sortPriority=57}
  n11 = Saturate(None=Divide(A=Subtract(A=@n9, B=@n10), B=ScalarParameter() {defaultValue=300, group=Knish | WPO, parameterName=Knish - WPO Fade Length, sortPriority=58}))
  n12 = Multiply(
   A=FunctionInput(Preview=Constant() {r=1}) {bUsePreviewValueAsDefault=true, inputName=StrengthMult, inputType=FunctionInput_Scalar, sortPriority=1}
   B=LinearInterpolate(B=OneMinus(None=@n11), Alpha=ScalarParameter() {group=Knish | WPO, parameterName=Knish - Use WPO Distance Fade, sortPriority=57}) {constA=1}
  )
  n13 = Multiply(
   A=Multiply(
    A=MaterialFunctionCall(X=@n5, Y=@n5, Z=@n6) {materialFunction=&MakeFloat3}.Result
    B=Multiply(A=@n8, B=@n12)
   )
   B=ScalarParameter() {group=Knish | WPO, parameterName=Knish - WPO Distortion Strength, sortPriority=54}
  )
  FunctionOutput(None=SetMaterialAttributes!(MaterialAttributes=@n1.MaterialAttributes, WorldPositionOffset=Add(A=@n1.WorldPositionOffset, B=@n13)) {attributeSetTypes [#1]})
  LinearInterpolate(
   B=@n13
   Alpha=ScalarParameter() {bUseCustomPrimitiveData=true, group=Knish | Dissolve, parameterName=CustomPrim Use Distance FX, sortPriority=40}
  )
 ]}
