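# Material function for the KnishClamp NPC's inner cloak, adding an animated galaxy surface with a glowing gradient.
# Two nebula layers and a star layer use virtual-box triplanar projection for depth, with time-of-day brightness and desaturation.
# Fresnel emissive and channel-selected texture masks shape the effect over the incoming material attributes.
# Pre-skinned local Z drives gradient masks, one blending toward flat EmissiveColor, before roughness and specular are forced fully matte.
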MaterialFunction /BRCosmetics/Characters/Player/Male/Medium/Bodies/M_MED_KnishClamp_NPC/MaterialFunctions/MF_KnishClamp_InnerCloak
graph {nodes [
  n1 = Multiply(A=OneMinus(None=NamedRerouteUsage()), B=NamedRerouteUsage())
  n2 = StaticSwitchParameter(True=Multiply(A=NamedRerouteUsage(), B=@n1), False=@n1) {group=Galaxy, parameterName=UseGalaxyMask, sortPriority=-8}
  Add(A=@n2, B=Multiply(A=NamedRerouteUsage(), B=NamedRerouteUsage()))
  Constant()
  n3 = FunctionInput() {inputName=PreSkinnedLocalPos, sortPriority=4}
  n4 = ScalarParameter() {defaultValue=100, group=Knish, parameterName=Knish - PreskinnedPosDistance}
  n5 = ComponentMask(None=Subtract(A=Divide(A=@n3, B=@n4) {constB=100}, B=ScalarParameter() {group=Knish, parameterName=Knish - PreskinnedPosOffset})) {b=true}
  n6 = LinearInterpolate(
   B=NamedRerouteDeclaration(None=Saturate(None=@n5)) {name=PreskinnedGradient, nodeColor=<rgba 0.3828 1 0 1>}
   Alpha=ScalarParameter() {group=Knish, parameterName=Knish - Gradient GlowAmount}
  ) {constA=1}
  n7 = VectorParameter() {group=Galaxy, parameterName=EmissiveColor, sortPriority=91, defaultValue=<rgba 0.0885 0.6921 1 1>}
  n8 = BlendMaterialAttributes(A=NamedRerouteUsage(), B=NamedRerouteUsage(), Alpha=NamedRerouteUsage())
  n9 = SetMaterialAttributes!(MaterialAttributes=@n8, BaseColor=@n7.RGB, EmissiveColor=@n7.RGB) {attributeSetTypes [#1, #2]}
  FunctionOutput(None=SetMaterialAttributes!(MaterialAttributes=BlendMaterialAttributes(A=@n9, B=@n8, Alpha=@n6), Roughness=Constant() {r=1}, Specular=Constant()) {attributeSetTypes [#3, #4]})
  n10 = Fresnel(ExponentIn=ScalarParameter() {defaultValue=3, group=Galaxy, parameterName=EmissiveFresnelExponent, sortPriority=90}) {baseReflectFraction=0}
  n11 = ScalarParameter() {defaultValue=1, group=Galaxy, parameterName=EmissiveFresnelIntensity, sortPriority=92}
  n12 = NamedRerouteUsage()
  n13 = Power(Base=NamedRerouteUsage(), Exp=ScalarParameter() {defaultValue=1, group=Galaxy, parameterName=EmissiveMaskExponent, sortPriority=93})
  n14 = Saturate(None=Add(A=ComponentMask(None=@n12) {g=true}, B=@n13))
  n15 = ScalarParameter() {group=Galaxy, parameterName=UseEmissiveMaskOnFresnel, sortPriority=92}
  n16 = NamedRerouteDeclaration(None=Multiply(A=Multiply(A=@n10, B=@n11), B=LinearInterpolate(B=@n14, Alpha=@n15) {constA=1})) {name=FresnelMask, nodeColor=<rgba 0 0.7891 1 1>}
  n17 = VectorParameter() {group=Galaxy, parameterName=EmissiveColor, sortPriority=91, defaultValue=<rgba 0.0885 0.6921 1 1>}
  n18 = Multiply(A=@n13, B=ScalarParameter() {group=Galaxy, parameterName=EmissiveMaskIntensity, sortPriority=92})
  n19 = StaticSwitchParameter(True=NamedRerouteUsage(), False=@n17.RGB) {group=Galaxy, parameterName=UseEmissiveMap, sortPriority=89}
  NamedRerouteDeclaration(None=Add(A=Multiply(A=@n16, B=@n17.RGB), B=Multiply(A=@n18, B=@n19))) {name=Fresnel, nodeColor=<rgba 0.8047 0 1 1>}
  n20 = FunctionInput() {inputName=Material, inputType=FunctionInput_MaterialAttributes}
  n21 = GetMaterialAttributes(None=@n20) {attributeGetTypes [
    #3
    #5
    #4
    #2
    #1
   ]}
  n22 = NamedRerouteDeclaration(None=ScalarParameter() {group=Galaxy, parameterName=UseIncomingMatParams}) {name=UseIncomingMatParams, nodeColor=<rgba 0.875 0 1 1>}
  NamedRerouteDeclaration(None=LinearInterpolate(A=ScalarParameter() {defaultValue=0.2, group=Galaxy, parameterName=Galaxy_Roughness, sortPriority=52}, B=@n21.Roughness, Alpha=@n22) {constB=0}) {name=OUTRoughness, nodeColor=<rgba 0 0.4844 1 1>}
  NamedRerouteDeclaration(None=LinearInterpolate(A=ScalarParameter() {defaultValue=0.1, group=Galaxy, parameterName=Galaxy_Metallic, sortPriority=50}, B=@n21.Metallic, Alpha=@n22) {constB=0}) {name=OUTMetallic, nodeColor=<rgba 0 0.4844 1 1>}
  n23 = TextureObjectParameter() {
   group=Galaxy
   parameterName=StarsTexture
   samplerType=SAMPLERTYPE_Masks
   sortPriority=80
   texture=/Game/Characters/Player/Male/Medium/Bodies/M_Med_Soldier_04/Skins/Galaxy/Textures/Celestial_T_Stars
  }
  n24 = MaterialFunctionCall!(
   Texture=@n23
   Depth=ScalarParameter() {defaultValue=150, group=Galaxy, parameterName=Stars_Depth, sortPriority=81}
   Tile=ScalarParameter() {defaultValue=2, group=Galaxy, parameterName=Stars_Tile, sortPriority=82}
   BlendContrast=ScalarParameter() {defaultValue=1, group=Galaxy, parameterName=Stars_BlendContrast, sortPriority=83}
   SpeedXY=ComponentMask(None=VectorParameter() {group=Galaxy, parameterName=Stars_Speed, sortPriority=84, defaultValue=<rgba 0.002 -0.03 0 1>}.RGB) {g=true, r=true}
   ActorOffset=NamedRerouteUsage()
  ) {desc=3 TextureSampler, materialFunction=/Game/Global/MaterialFunctions/MF_TriplanarVirtualBox}
  NamedRerouteDeclaration(None=Multiply(
    A=ComponentMask(None=@n24.Result) {b=true}
    B=ScalarParameter() {defaultValue=1, group=Galaxy, parameterName=Stars_Brightness, sortPriority=84}
   )) {name=Stars, nodeColor=<rgba 0 1 0.5078 1>}
  n25 = TextureObjectParameter() {
   group=Galaxy
   parameterName=Galaxy_01_Texture
   sortPriority=58
   texture=/Game/Characters/Player/Male/Medium/Bodies/M_Med_Soldier_04/Skins/Galaxy/Textures/T_FN_Nebula_style_blue
  }
  n26 = ScalarParameter() {defaultValue=80, group=Galaxy, parameterName=Galaxy_01_Depth, sortPriority=61}
  n27 = ScalarParameter() {defaultValue=0.4, group=Galaxy, parameterName=Galaxy_01_Tile, sortPriority=62}
  n28 = ScalarParameter() {defaultValue=0.4, group=Galaxy, parameterName=Galaxy_01_BlendContrast, sortPriority=63}
  n29 = VectorParameter() {group=Galaxy, parameterName=Galaxy_01_Speed, sortPriority=64, defaultValue=<rgba 0.002 -0.0025 0 1>}
  n30 = ComponentMask(None=@n29.RGB) {g=true, r=true}
  n31 = NamedRerouteDeclaration(None=VectorParameter() {group=Galaxy, parameterName=Galaxy_Offset, sortPriority=59, defaultValue=<rgba 0 0 0 1>}.RGB) {name=ActorOffset, nodeColor=<rgba 0.9688 1 0 1>}
  n32 = Multiply(
   A=MaterialFunctionCall(
    Hue Shift Percentage=ScalarParameter() {group=Galaxy, parameterName=Galaxy_01_HueShift, sortPriority=64}
    Texture=MaterialFunctionCall(Texture=@n25, Depth=@n26, Tile=@n27, BlendContrast=@n28, SpeedXY=@n30, ActorOffset=@n31) {materialFunction=&MF_TriplanarVirtualBox}.Result
   ) {materialFunction=&HueShift}.Result
   B=ScalarParameter() {defaultValue=1, group=Galaxy, parameterName=Galaxy_01_Brightness, sortPriority=65}
  )
  n33 = ScalarParameter() {defaultValue=1.75, group=Galaxy, parameterName=Galaxy_01_Exp, sortPriority=66}
  NamedRerouteDeclaration(None=Desaturation(
    None=Power(Base=@n32, Exp=@n33)
    Fraction=ScalarParameter() {defaultValue=0.5, group=Galaxy, parameterName=Galaxy_01_Desaturation, sortPriority=67}
   )) {name=Galaxy_01, nodeColor=<rgba 1 0.2578 0 1>}
  NamedRerouteDeclaration(None=LinearInterpolate(A=ScalarParameter() {defaultValue=0.5, group=Galaxy, parameterName=Galaxy_Spec, sortPriority=51}, B=@n21.Specular, Alpha=@n22) {constB=0}) {name=OUTSpec, nodeColor=<rgba 0 0.4844 1 1>}
  NamedRerouteDeclaration(None=@n21.EmissiveColor) {name=IN_Emissive, nodeColor=<rgba 0.4766 1 0 1>}
  NamedRerouteDeclaration(None=@n19) {name=FresnelEmissive, nodeColor=<rgba 0 1 0.7656 1>}
  n34 = TextureSampleParameter2D() {
   group=Galaxy
   parameterName=MaskTexture_DetailedControl
   samplerType=SAMPLERTYPE_Masks
   sortPriority=-20
   texture=&T_WhiteMask
  }
  n35 = ChannelMaskParameter(None=@n34.RGB) {group=Galaxy, maskChannel=Blue, parameterName=GalaxyMaskChannel, sortPriority=-7, defaultValue=<rgba 0 0 1 0>}
  NamedRerouteDeclaration(None=LinearInterpolate(B=@n35.R, Alpha=ScalarParameter() {defaultValue=1, group=Galaxy, parameterName=GalaxyMaskStrength, sortPriority=-6}) {constA=1}) {name=GalaxyMask, nodeColor=<rgba 0.3125 0 1 1>}
  NamedRerouteDeclaration(None=@n2) {name=OUT_Emissive, nodeColor=<rgba 1 0 0.1875 1>}
  n36 = VectorParameter() {group=Galaxy, parameterName=EmissiveColor, sortPriority=91, defaultValue=<rgba 0.0885 0.6921 1 1>}
  n37 = VectorParameter() {group=Galaxy, parameterName=Galaxy_BaseColor, sortPriority=49, defaultValue=<rgba 0 0 0 1>}
  n38 = NamedRerouteUsage()
  NamedRerouteDeclaration(None=LinearInterpolate(A=LinearInterpolate(A=@n36.RGB, B=@n37.RGB, Alpha=@n38), B=@n21.EmissiveColor, Alpha=@n22) {constB=0}) {name=OUTBaseColor, nodeColor=<rgba 1 0 0.6562 1>}
  n39 = FunctionInput() {bUsePreviewValueAsDefault=true, inputName=TechArtMask, sortPriority=1, previewValue=<1 1 1 1>}
  NamedRerouteDeclaration(None=@n39) {name=TechArtMask, nodeColor=<rgba 0 1 0.6953 1>}
  n40 = ScalarParameter() {defaultValue=100, group=Knish, parameterName=Knish - FresnelGradientDistance}
  n41 = ComponentMask(None=Subtract(A=Divide(A=@n3, B=@n40) {constB=100}, B=ScalarParameter() {group=Knish, parameterName=Knish - FresnelGradientPosOffset})) {b=true}
  NamedRerouteDeclaration(None=OneMinus(None=Saturate(None=@n41))) {name=FresnelGradientMask, nodeColor=<rgba 1 0.5391 0 1>}
  n42 = SetMaterialAttributes!(
   MaterialAttributes=NamedRerouteUsage()
   BaseColor=NamedRerouteUsage()
   Metallic=NamedRerouteUsage()
   Specular=NamedRerouteUsage()
   Roughness=NamedRerouteUsage()
   EmissiveColor=NamedRerouteUsage()
  ) {attributeSetTypes [
    #1
    #5
    #4
    #3
    #2
   ]}
  NamedRerouteDeclaration(None=@n42) {name=MaterialOUT, nodeColor=<rgba 0 1 0.625 1>}
  NamedRerouteDeclaration(None=ChannelMaskParameter(None=@n34.RGB) {group=Galaxy, parameterName=AlphaChannel, sortPriority=-10}.R) {name=Alpha, nodeColor=<rgba 0.7109 1 0 1>}
  NamedRerouteDeclaration(None=@n20) {name=IN_Material, nodeColor=<rgba 1 0 0.9609 1>}
  n43 = ScalarParameter() {group=Galaxy, parameterName=Galaxy_02_HueShift, sortPriority=74}
  n44 = TextureObjectParameter() {
   group=Galaxy
   parameterName=Galaxy_02_Texture
   sortPriority=70
   texture=/Game/Characters/Player/Male/Medium/Bodies/M_Med_Soldier_04/Skins/Galaxy/Textures/T_FN_Nebula_style2
  }
  n45 = ScalarParameter() {defaultValue=35, group=Galaxy, parameterName=Galaxy_02_Depth, sortPriority=71}
  n46 = ScalarParameter() {defaultValue=0.4, group=Galaxy, parameterName=Galaxy_02_Tile, sortPriority=72}
  n47 = ScalarParameter() {defaultValue=1, group=Galaxy, parameterName=Galaxy_02_BlendContrast, sortPriority=73}
  n48 = VectorParameter() {group=Galaxy, parameterName=Galaxy_02_Speed, sortPriority=74, defaultValue=<rgba -0.002 -0.0065 0 1>}
  n49 = NamedRerouteUsage()
  n50 = MaterialFunctionCall(
   Texture=@n44
   Depth=@n45
   Tile=Multiply(B=@n46) {constA=1.3}
   BlendContrast=@n47
   SpeedXY=ComponentMask(None=@n48.RGB) {g=true, r=true}
   ActorOffset=@n49
  ) {materialFunction=&MF_TriplanarVirtualBox}
  n51 = MaterialFunctionCall(
   Texture=@n44
   Depth=@n45
   Tile=@n46
   BlendContrast=@n47
   SpeedXY=MaterialFunctionCall(X=@n48.B, Y=@n48.B) {materialFunction=&MakeFloat2}.Result
   ActorOffset=@n49
  ) {materialFunction=&MF_TriplanarVirtualBox}
  n52 = MaterialFunctionCall(X=ComponentMask(None=Multiply(A=@n50.Result, B=@n51.Result)) {r=true}) {materialFunction=&MakeFloat3}
  NamedRerouteDeclaration(None=Multiply(
    A=MaterialFunctionCall(Hue Shift Percentage=@n43, Texture=@n52.Result) {materialFunction=&HueShift}.Result
    B=ScalarParameter() {defaultValue=0.5, group=Galaxy, parameterName=Galaxy_02_Brightness, sortPriority=75}
   )) {name=Galaxy_02, nodeColor=<rgba 0.0781 0 1 1>}
  n53 = Add(A=NamedRerouteUsage(), B=NamedRerouteUsage())
  n54 = ChannelMaskParameter(None=NamedRerouteDeclaration(None=Saturate(None=@n53)) {name=Galaxies, nodeColor=<rgba 1 0 0.2578 1>}) {group=Galaxy, maskChannel=Blue, parameterName=StarMaskChannel, sortPriority=84, defaultValue=<rgba 0 0 1 0>}
  NamedRerouteDeclaration(None=Multiply(A=@n54.R, B=ScalarParameter() {defaultValue=3, group=Galaxy, parameterName=GalaxyStarMaskStrength, sortPriority=85}) {constB=2}) {name=StarMask, nodeColor=<rgba 0.9922 1 0 1>}
  n55 = LinearInterpolate(A=NamedRerouteUsage(), Alpha=NamedRerouteUsage()) {constB=0}
  n56 = Multiply(
   A=NamedRerouteUsage()
   B=MaterialFunctionCall(
    Value_Morning=ScalarParameter() {defaultValue=1, group=Galaxy, parameterName=Galaxy_Brightness_Morning, sortPriority=100}
    Value_Day=ScalarParameter() {defaultValue=1, group=Galaxy, parameterName=Galaxy_Brightness_Day, sortPriority=101}
    Value_Evening=ScalarParameter() {defaultValue=1, group=Galaxy, parameterName=Galaxy_Brightness_Evening, sortPriority=102}
    Value_Night=ScalarParameter() {defaultValue=1, group=Galaxy, parameterName=Galaxy_Brightness_Night, sortPriority=103}
   ) {materialFunction=&Scalar_Time_of_Day}.Result
  )
  n57 = MaterialFunctionCall(
   Value_Morning=ScalarParameter() {group=Galaxy, parameterName=Galaxy_Desat_Morning, sortPriority=110}
   Value_Day=ScalarParameter() {group=Galaxy, parameterName=Galaxy_Desat_Day, sortPriority=111}
   Value_Evening=ScalarParameter() {group=Galaxy, parameterName=Galaxy_Desat_Evening, sortPriority=112}
   Value_Night=ScalarParameter() {group=Galaxy, parameterName=Galaxy_Desat_Night, sortPriority=113}
  ) {materialFunction=&Scalar_Time_of_Day}
  NamedRerouteDeclaration(None=Add(A=Saturate(None=@n55), B=Desaturation(None=@n56, Fraction=@n57.Result))) {name=GalaxiesAndStars, nodeColor=<rgba 0 1 0.9062 1>}
  n58 = ChannelMaskParameter(None=@n34.RGB) {group=Galaxy, maskChannel=Green, parameterName=EmissiveMaskChannel, sortPriority=-9, defaultValue=<rgba 0 1 0 0>}
  NamedRerouteDeclaration(None=@n58.R) {name=EmissiveMask, nodeColor=<rgba 0 0.4375 1 1>}
 ]}
