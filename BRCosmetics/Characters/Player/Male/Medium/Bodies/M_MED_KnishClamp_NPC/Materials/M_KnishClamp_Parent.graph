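# Masked parent material for the KnishClamp NPC character body, built on the shared 2023 character PreFX and PostFX functions with support for several shading models.
# The main chain blacks out areas by the tech-art mask's green channel, then applies near-distance fade and distortion before PostFX.
# Separate branches with no visible consumer hold Fresnel and blackout blends, a two-sphere emissive mask and an inner-cloak function.
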Material /BRCosmetics/Characters/Player/Male/Medium/Bodies/M_MED_KnishClamp_NPC/Materials/M_KnishClamp_Parent
graph {
 properties {
  bUseMaterialAttributes=true
  bUsedWithMeshDeformer=true
  bUsedWithMorphTargets=true
  bUsedWithNanite=true
  bUsedWithSkeletalMesh=true
  blendMode=BLEND_Masked
  pixelDepthOffsetMode=PDOM_Legacy
  shadingModel=MSM_FromMaterialExpression
  usedShadingModels=Default Lit | Subsurface | Clear Coat | Subsurface Profile | Cloth
 }
 nodes [
  n1 = MaterialFunctionCall() {materialFunction=&MF_CharacterParent_2023_PreFX}
  n2 = GetMaterialAttributes(None=@n1.Result) {attributeGetTypes [#1, #2, #3]}
  n3 = MaterialFunctionCall(Exponent=ScalarParameter() {defaultValue=8, group=Knish, parameterName=Knish - FresnelExp, sortPriority=25}) {materialFunction=&MF_Fresnel_FX}
  n4 = LinearInterpolate(A=@n3.Fresnel, B=@n3.Fresnel, Alpha=ScalarParameter() {group=Knish, parameterName=Knish - Fresnel Type Switch, sliderMax=1, sortPriority=25})
  n5 = LinearInterpolate(B=@n2.Base Color, Alpha=@n4)
  n6 = LinearInterpolate(B=@n2.Roughness, Alpha=@n4) {constA=1}
  n7 = LinearInterpolate(B=@n2.Specular, Alpha=@n4)
  BlendMaterialAttributes(
   A=BlendMaterialAttributes(
    A=@n1.Result
    B=SetMaterialAttributes!(MaterialAttributes=@n2.MaterialAttributes, Base Color=@n5, Roughness=@n6, Specular=@n7) {attributeSetTypes [#1, #2, #3]}
    Alpha=ScalarParameter() {group=Knish, parameterName=Knish - Use Fresnel, sliderMax=1, sortPriority=20}
   )
   B=SetMaterialAttributes(MaterialAttributes=@n2.MaterialAttributes, Base Color=Constant(), Roughness=Constant() {r=1}, Specular=Constant()) {attributeSetTypes [#1, #2, #3]}
   Alpha=ScalarParameter() {group=Knish, parameterName=Knish - Use Blackout, sliderMax=1, sortPriority=21}
  )
  n8 = VectorParameter() {parameterName=SphereLocation, defaultValue=<rgba 0 0 0 100>}
  n9 = Reroute() {desc=Preskinned}
  n10 = Add(A=@n9, B=ComponentMask(None=@n8.RGB) {b=true, g=true, r=true})
  n11 = LocalPosition()
  n12 = ScalarParameter() {defaultValue=0.8, parameterName=Hardness}
  n13 = VectorParameter() {parameterName=SphereLocation, defaultValue=<rgba 0 0 0 100>}
  n14 = Add(A=@n9, B=Multiply(A=@n13.RGB, B=Constant3Vector() {constant=<rgba -1 1 1 0>}) {constB=-1})
  n15 = LocalPosition()
  BlendMaterialAttributes(
   B=MakeMaterialAttributes!(
    BaseColor
    Metallic
    Specular
    Roughness
    Anisotropy
    EmissiveColor=Constant3Vector() {constant=<rgba 0 1 0 0>}
    Opacity
    OpacityMask
    Normal
    Tangent
    WorldPositionOffset
    SubsurfaceColor
    ClearCoat
    ClearCoatRoughness
    AmbientOcclusion
    Refraction
    CustomizedUVs_0
    CustomizedUVs_1
    CustomizedUVs_2
    CustomizedUVs_3
    CustomizedUVs_4
    CustomizedUVs_5
    CustomizedUVs_6
    CustomizedUVs_7
    PixelDepthOffset
    ShadingModel
    Displacement
   )
   Alpha=Max(
    A=SphereMask(A=@n10, B=@n11.XYZ, Radius=@n8.A, Hardness=@n12) {attenuationRadius=150, hardnessPercent=0.8}
    B=SphereMask(A=@n14, B=@n15.XYZ, Radius=@n13.A, Hardness=@n12) {attenuationRadius=150, hardnessPercent=0.8}
   )
  )
  BumpOffset()
  n16 = BlendMaterialAttributes(
   A=@n1.Result
   B=MakeMaterialAttributes(BaseColor=Constant(), Specular=Constant(), Roughness=Constant() {r=1})
   Alpha=ComponentMask(None=@n1.Tech Art Mask) {g=true}
  )
  MaterialFunctionCall(
   Material=MaterialFunctionCall!(
    In=MaterialFunctionCall!(In=@n16) {materialFunction=/BRCosmetics/Characters/Player/Male/Medium/Bodies/M_MED_KnishClamp_NPC/MaterialFunctions/MF_KnishClamp_NearDistanceFade}.Result
    StrengthMult
    PreskinnedLocalPos=@n1.Pre skinned Local Position
   ) {materialFunction=/BRCosmetics/Characters/Player/Male/Medium/Bodies/M_MED_KnishClamp_NPC/MaterialFunctions/MF_KnishClamp_Distortion}.Result
   M=@n1.M
   Pre Skinned Local Position=@n1.Pre skinned Local Position
   Distance Blend=@n1.M
   Fresnel=@n1.M
  ) {materialFunction=&MF_CharacterParent_2023_PostFX}
  MaterialFunctionCall!(Material=@n1.Result, TechArtMask=@n1.Tech Art Mask, PreSkinnedLocalPos=@n1.Pre skinned Local Position) {materialFunction=/BRCosmetics/Characters/Player/Male/Medium/Bodies/M_MED_KnishClamp_NPC/MaterialFunctions/MF_KnishClamp_InnerCloak}
 ]
}
