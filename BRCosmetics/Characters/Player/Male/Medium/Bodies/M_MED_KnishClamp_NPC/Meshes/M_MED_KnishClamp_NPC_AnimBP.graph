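# Animation Blueprint for the KnishClamp NPC body that adds procedural sway and physics to its cloak, atop the pose copied from the attached parent mesh.
# Each update, time-driven random rotators, apparently Perlin-based, set additive bone-space rotations for the lower cloak bones.
# That cached pose feeds serial AnimDynamics nodes across seven five-bone strips, with damped angular springs, reduced gravity and alpha tapering from tip to root.
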AnimBlueprint /BRCosmetics/Characters/Player/Male/Medium/Bodies/M_MED_KnishClamp_NPC/Meshes/M_MED_KnishClamp_NPC_AnimBP
target_skeleton=&/M_MED_KnishClamp_NPC_Skeleton
graphs [
 AnimGraph {nodes [
   n1 = Anim Dynamics - Bone: dyn_cloak_front_l_05(ComponentPose=Local To Component(LocalPose=Use cached pose 'PostRotation'()), Alpha=1, ExternalForce="0,0,0") {
    angularDampingOverride=0.9999
    angularSpringConstant=5000
    bAngularSpring=true
    bOverrideAngularDamping=true
    bUseGravityOverride=true
    gravityScale=0.35
    simulationSpace=RootRelative
    alphaBoolBlend {blendOption=Linear}
    boundBone {boneName=dyn_cloak_front_l_05}
    componentLinearAccScale=<0.05 0.05 0.05>
    componentLinearVelScale=<0.05 0.05 0.05>
    gravityOverride {z=-980}
    physicsBodyDefinitions [{
      boundBone {boneName=dyn_cloak_front_l_05}
      boxExtents=<150 150 150>
      constraintSetup {angularLimitsMax=<0 15 60>, angularLimitsMin=<0 -60 -60>, angularTarget=<1 0 0>}
      localJointOffset=<0 0 0>
     }]
   }
   n2 = Anim Dynamics - Bone: dyn_cloak_front_l_04(ComponentPose=@n1, Alpha=1, ExternalForce="0,0,0") {
    angularDampingOverride=0.9999
    angularSpringConstant=5000
    bAngularSpring=true
    bOverrideAngularDamping=true
    bUseGravityOverride=true
    gravityScale=0.35
    simulationSpace=RootRelative
    alphaBoolBlend {blendOption=Linear}
    alphaScaleBiasClamp {scale=0.75}
    boundBone {boneName=dyn_cloak_front_l_04}
    componentLinearAccScale=<0.05 0.05 0.05>
    componentLinearVelScale=<0.05 0.05 0.05>
    gravityOverride {z=-980}
    physicsBodyDefinitions [{
      boundBone {boneName=dyn_cloak_front_l_04}
      boxExtents=<150 150 150>
      constraintSetup {angularLimitsMax=<0 15 60>, angularLimitsMin=<0 -60 -60>, angularTarget=<1 0 0>}
      localJointOffset=<0 0 0>
     }]
   }
   n3 = Anim Dynamics - Bone: dyn_cloak_front_l_03(ComponentPose=@n2, Alpha=1, ExternalForce="0,0,0") {
    angularDampingOverride=0.9999
    angularSpringConstant=5000
    bAngularSpring=true
    bOverrideAngularDamping=true
    bUseGravityOverride=true
    gravityScale=0.35
    simulationSpace=RootRelative
    alphaBoolBlend {blendOption=Linear}
    alphaScaleBiasClamp {scale=0.75}
    boundBone {boneName=dyn_cloak_front_l_03}
    componentLinearAccScale=<0.05 0.05 0.05>
    componentLinearVelScale=<0.05 0.05 0.05>
    gravityOverride {z=-980}
    physicsBodyDefinitions [{
      boundBone {boneName=dyn_cloak_front_l_03}
      boxExtents=<150 150 150>
      constraintSetup {angularLimitsMax=<0 15 60>, angularLimitsMin=<0 -60 -60>, angularTarget=<1 0 0>}
      localJointOffset=<0 0 0>
     }]
   }
   n4 = Anim Dynamics - Bone: dyn_cloak_front_l_02(ComponentPose=@n3, Alpha=1, ExternalForce="0,0,0") {
    angularDampingOverride=0.9999
    angularSpringConstant=5000
    bAngularSpring=true
    bOverrideAngularDamping=true
    bUseGravityOverride=true
    gravityScale=0.35
    simulationSpace=RootRelative
    alphaBoolBlend {blendOption=Linear}
    alphaScaleBiasClamp {scale=0.5}
    boundBone {boneName=dyn_cloak_front_l_02}
    componentLinearAccScale=<0.05 0.05 0.05>
    componentLinearVelScale=<0.05 0.05 0.05>
    gravityOverride {z=-980}
    physicsBodyDefinitions [{
      boundBone {boneName=dyn_cloak_front_l_02}
      boxExtents=<150 150 150>
      constraintSetup {angularLimitsMax=<0 15 60>, angularLimitsMin=<0 -60 -60>, angularTarget=<1 0 0>}
      localJointOffset=<0 0 0>
     }]
   }
   n5 = Anim Dynamics - Bone: dyn_cloak_front_l_01(ComponentPose=@n4, Alpha=1, ExternalForce="0,0,0") {
    angularDampingOverride=0.9999
    angularSpringConstant=5000
    bAngularSpring=true
    bOverrideAngularDamping=true
    bUseGravityOverride=true
    gravityScale=0.35
    simulationSpace=RootRelative
    alphaBoolBlend {blendOption=Linear}
    alphaScaleBiasClamp {scale=0.25}
    boundBone {boneName=dyn_cloak_front_l_01}
    componentLinearAccScale=<0.05 0.05 0.05>
    componentLinearVelScale=<0.05 0.05 0.05>
    gravityOverride {z=-980}
    physicsBodyDefinitions [{
      boundBone {boneName=dyn_cloak_front_l_01}
      boxExtents=<150 150 150>
      constraintSetup {angularLimitsMax=<0 15 60>, angularLimitsMin=<0 -60 -60>, angularTarget=<1 0 0>}
      localJointOffset=<0 0 0>
     }]
   }
   n6 = Anim Dynamics - Bone: dyn_cloak_side_l_05(ComponentPose=@n5, Alpha=1, ExternalForce="0,0,0") {
    angularDampingOverride=0.9999
    angularSpringConstant=5000
    bAngularSpring=true
    bOverrideAngularDamping=true
    bUseGravityOverride=true
    gravityScale=0.35
    simulationSpace=RootRelative
    alphaBoolBlend {blendOption=Linear}
    boundBone {boneName=dyn_cloak_side_l_05}
    componentLinearAccScale=<0.05 0.05 0.05>
    componentLinearVelScale=<0.05 0.05 0.05>
    gravityOverride {z=-980}
    physicsBodyDefinitions [{
      boundBone {boneName=dyn_cloak_side_l_05}
      boxExtents=<150 150 150>
      constraintSetup {angularLimitsMax=<0 15 60>, angularLimitsMin=<0 -60 -60>, angularTarget=<1 0 0>}
      localJointOffset=<0 0 0>
     }]
   }
   n7 = Anim Dynamics - Bone: dyn_cloak_side_l_04(ComponentPose=@n6, Alpha=1, ExternalForce="0,0,0") {
    angularDampingOverride=0.9999
    angularSpringConstant=5000
    bAngularSpring=true
    bOverrideAngularDamping=true
    bUseGravityOverride=true
    gravityScale=0.35
    simulationSpace=RootRelative
    alphaBoolBlend {blendOption=Linear}
    alphaScaleBiasClamp {scale=0.75}
    boundBone {boneName=dyn_cloak_side_l_04}
    componentLinearAccScale=<0.05 0.05 0.05>
    componentLinearVelScale=<0.05 0.05 0.05>
    gravityOverride {z=-980}
    physicsBodyDefinitions [{
      boundBone {boneName=dyn_cloak_side_l_04}
      boxExtents=<150 150 150>
      constraintSetup {angularLimitsMax=<0 15 60>, angularLimitsMin=<0 -60 -60>, angularTarget=<1 0 0>}
      localJointOffset=<0 0 0>
     }]
   }
   n8 = Anim Dynamics - Bone: dyn_cloak_side_l_03(ComponentPose=@n7, Alpha=1, ExternalForce="0,0,0") {
    angularDampingOverride=0.9999
    angularSpringConstant=5000
    bAngularSpring=true
    bOverrideAngularDamping=true
    bUseGravityOverride=true
    gravityScale=0.35
    simulationSpace=RootRelative
    alphaBoolBlend {blendOption=Linear}
    alphaScaleBiasClamp {scale=0.75}
    boundBone {boneName=dyn_cloak_side_l_03}
    componentLinearAccScale=<0.05 0.05 0.05>
    componentLinearVelScale=<0.05 0.05 0.05>
    gravityOverride {z=-980}
    physicsBodyDefinitions [{
      boundBone {boneName=dyn_cloak_side_l_03}
      boxExtents=<150 150 150>
      constraintSetup {angularLimitsMax=<0 15 60>, angularLimitsMin=<0 -60 -60>, angularTarget=<1 0 0>}
      localJointOffset=<0 0 0>
     }]
   }
   n9 = Anim Dynamics - Bone: dyn_cloak_side_l_02(ComponentPose=@n8, Alpha=1, ExternalForce="0,0,0") {
    angularDampingOverride=0.9999
    angularSpringConstant=5000
    bAngularSpring=true
    bOverrideAngularDamping=true
    bUseGravityOverride=true
    gravityScale=0.35
    simulationSpace=RootRelative
    alphaBoolBlend {blendOption=Linear}
    alphaScaleBiasClamp {scale=0.5}
    boundBone {boneName=dyn_cloak_side_l_02}
    componentLinearAccScale=<0.05 0.05 0.05>
    componentLinearVelScale=<0.05 0.05 0.05>
    gravityOverride {z=-980}
    physicsBodyDefinitions [{
      boundBone {boneName=dyn_cloak_side_l_02}
      boxExtents=<150 150 150>
      constraintSetup {angularLimitsMax=<0 15 60>, angularLimitsMin=<0 -60 -60>, angularTarget=<1 0 0>}
      localJointOffset=<0 0 0>
     }]
   }
   n10 = Anim Dynamics - Bone: dyn_cloak_side_l_01(ComponentPose=@n9, Alpha=1, ExternalForce="0,0,0") {
    angularDampingOverride=0.9999
    angularSpringConstant=5000
    bAngularSpring=true
    bOverrideAngularDamping=true
    bUseGravityOverride=true
    gravityScale=0.35
    simulationSpace=RootRelative
    alphaBoolBlend {blendOption=Linear}
    alphaScaleBiasClamp {scale=0.25}
    boundBone {boneName=dyn_cloak_side_l_01}
    componentLinearAccScale=<0.05 0.05 0.05>
    componentLinearVelScale=<0.05 0.05 0.05>
    gravityOverride {z=-980}
    physicsBodyDefinitions [{
      boundBone {boneName=dyn_cloak_side_l_01}
      boxExtents=<150 150 150>
      constraintSetup {angularLimitsMax=<0 15 60>, angularLimitsMin=<0 -60 -60>, angularTarget=<1 0 0>}
      localJointOffset=<0 0 0>
     }]
   }
   n11 = Anim Dynamics - Bone: dyn_cloak_rear_l_05(ComponentPose=@n10, Alpha=1, ExternalForce="0,0,0") {
    angularDampingOverride=0.9999
    angularSpringConstant=5000
    bAngularSpring=true
    bOverrideAngularDamping=true
    bUseGravityOverride=true
    gravityScale=0.35
    simulationSpace=RootRelative
    alphaBoolBlend {blendOption=Linear}
    boundBone {boneName=dyn_cloak_rear_l_05}
    componentLinearAccScale=<0.05 0.05 0.05>
    componentLinearVelScale=<0.05 0.05 0.05>
    gravityOverride {z=-980}
    physicsBodyDefinitions [{
      boundBone {boneName=dyn_cloak_rear_l_05}
      boxExtents=<150 150 150>
      constraintSetup {angularLimitsMax=<0 15 60>, angularLimitsMin=<0 -60 -60>, angularTarget=<1 0 0>}
      localJointOffset=<0 0 0>
     }]
   }
   n12 = Anim Dynamics - Bone: dyn_cloak_rear_l_04(ComponentPose=@n11, Alpha=1, ExternalForce="0,0,0") {
    angularDampingOverride=0.9999
    angularSpringConstant=5000
    bAngularSpring=true
    bOverrideAngularDamping=true
    bUseGravityOverride=true
    gravityScale=0.35
    simulationSpace=RootRelative
    alphaBoolBlend {blendOption=Linear}
    alphaScaleBiasClamp {scale=0.75}
    boundBone {boneName=dyn_cloak_rear_l_04}
    componentLinearAccScale=<0.05 0.05 0.05>
    componentLinearVelScale=<0.05 0.05 0.05>
    gravityOverride {z=-980}
    physicsBodyDefinitions [{
      boundBone {boneName=dyn_cloak_rear_l_04}
      boxExtents=<150 150 150>
      constraintSetup {angularLimitsMax=<0 15 60>, angularLimitsMin=<0 -60 -60>, angularTarget=<1 0 0>}
      localJointOffset=<0 0 0>
     }]
   }
   n13 = Anim Dynamics - Bone: dyn_cloak_rear_l_03(ComponentPose=@n12, Alpha=1, ExternalForce="0,0,0") {
    angularDampingOverride=0.9999
    angularSpringConstant=5000
    bAngularSpring=true
    bOverrideAngularDamping=true
    bUseGravityOverride=true
    gravityScale=0.35
    simulationSpace=RootRelative
    alphaBoolBlend {blendOption=Linear}
    alphaScaleBiasClamp {scale=0.75}
    boundBone {boneName=dyn_cloak_rear_l_03}
    componentLinearAccScale=<0.05 0.05 0.05>
    componentLinearVelScale=<0.05 0.05 0.05>
    gravityOverride {z=-980}
    physicsBodyDefinitions [{
      boundBone {boneName=dyn_cloak_rear_l_03}
      boxExtents=<150 150 150>
      constraintSetup {angularLimitsMax=<0 15 60>, angularLimitsMin=<0 -60 -60>, angularTarget=<1 0 0>}
      localJointOffset=<0 0 0>
     }]
   }
   n14 = Anim Dynamics - Bone: dyn_cloak_rear_l_02(ComponentPose=@n13, Alpha=1, ExternalForce="0,0,0") {
    angularDampingOverride=0.9999
    angularSpringConstant=5000
    bAngularSpring=true
    bOverrideAngularDamping=true
    bUseGravityOverride=true
    gravityScale=0.35
    simulationSpace=RootRelative
    alphaBoolBlend {blendOption=Linear}
    alphaScaleBiasClamp {scale=0.5}
    boundBone {boneName=dyn_cloak_rear_l_02}
    componentLinearAccScale=<0.05 0.05 0.05>
    componentLinearVelScale=<0.05 0.05 0.05>
    gravityOverride {z=-980}
    physicsBodyDefinitions [{
      boundBone {boneName=dyn_cloak_rear_l_02}
      boxExtents=<150 150 150>
      constraintSetup {angularLimitsMax=<0 15 60>, angularLimitsMin=<0 -60 -60>, angularTarget=<1 0 0>}
      localJointOffset=<0 0 0>
     }]
   }
   n15 = Anim Dynamics - Bone: dyn_cloak_rear_l_01(ComponentPose=@n14, Alpha=1, ExternalForce="0,0,0") {
    angularDampingOverride=0.9999
    angularSpringConstant=5000
    bAngularSpring=true
    bOverrideAngularDamping=true
    bUseGravityOverride=true
    gravityScale=0.35
    simulationSpace=RootRelative
    alphaBoolBlend {blendOption=Linear}
    alphaScaleBiasClamp {scale=0.25}
    boundBone {boneName=dyn_cloak_rear_l_01}
    componentLinearAccScale=<0.05 0.05 0.05>
    componentLinearVelScale=<0.05 0.05 0.05>
    gravityOverride {z=-980}
    physicsBodyDefinitions [{
      boundBone {boneName=dyn_cloak_rear_l_01}
      boxExtents=<150 150 150>
      constraintSetup {angularLimitsMax=<0 15 60>, angularLimitsMin=<0 -60 -60>, angularTarget=<1 0 0>}
      localJointOffset=<0 0 0>
     }]
   }
   n16 = Anim Dynamics - Bone: dyn_cloak_rear_c_05(ComponentPose=@n15, Alpha=1, ExternalForce="0,0,0") {
    angularDampingOverride=0.9999
    angularSpringConstant=5000
    bAngularSpring=true
    bOverrideAngularDamping=true
    bUseGravityOverride=true
    gravityScale=0.35
    simulationSpace=RootRelative
    alphaBoolBlend {blendOption=Linear}
    boundBone {boneName=dyn_cloak_rear_c_05}
    componentLinearAccScale=<0.05 0.05 0.05>
    componentLinearVelScale=<0.05 0.05 0.05>
    gravityOverride {z=-980}
    physicsBodyDefinitions [{
      boundBone {boneName=dyn_cloak_rear_c_05}
      boxExtents=<150 150 150>
      constraintSetup {angularLimitsMax=<0 15 60>, angularLimitsMin=<0 -60 -60>, angularTarget=<1 0 0>}
      localJointOffset=<0 0 0>
     }]
   }
   n17 = Anim Dynamics - Bone: dyn_cloak_rear_c_04(ComponentPose=@n16, Alpha=1, ExternalForce="0,0,0") {
    angularDampingOverride=0.9999
    angularSpringConstant=5000
    bAngularSpring=true
    bOverrideAngularDamping=true
    bUseGravityOverride=true
    gravityScale=0.35
    simulationSpace=RootRelative
    alphaBoolBlend {blendOption=Linear}
    alphaScaleBiasClamp {scale=0.75}
    boundBone {boneName=dyn_cloak_rear_c_04}
    componentLinearAccScale=<0.05 0.05 0.05>
    componentLinearVelScale=<0.05 0.05 0.05>
    gravityOverride {z=-980}
    physicsBodyDefinitions [{
      boundBone {boneName=dyn_cloak_rear_c_04}
      boxExtents=<150 150 150>
      constraintSetup {angularLimitsMax=<0 15 60>, angularLimitsMin=<0 -60 -60>, angularTarget=<1 0 0>}
      localJointOffset=<0 0 0>
     }]
   }
   n18 = Anim Dynamics - Bone: dyn_cloak_rear_c_03(ComponentPose=@n17, Alpha=1, ExternalForce="0,0,0") {
    angularDampingOverride=0.9999
    angularSpringConstant=5000
    bAngularSpring=true
    bOverrideAngularDamping=true
    bUseGravityOverride=true
    gravityScale=0.35
    simulationSpace=RootRelative
    alphaBoolBlend {blendOption=Linear}
    alphaScaleBiasClamp {scale=0.75}
    boundBone {boneName=dyn_cloak_rear_c_03}
    componentLinearAccScale=<0.05 0.05 0.05>
    componentLinearVelScale=<0.05 0.05 0.05>
    gravityOverride {z=-980}
    physicsBodyDefinitions [{
      boundBone {boneName=dyn_cloak_rear_c_03}
      boxExtents=<150 150 150>
      constraintSetup {angularLimitsMax=<0 15 60>, angularLimitsMin=<0 -60 -60>, angularTarget=<1 0 0>}
      localJointOffset=<0 0 0>
     }]
   }
   n19 = Anim Dynamics - Bone: dyn_cloak_rear_c_02(ComponentPose=@n18, Alpha=1, ExternalForce="0,0,0") {
    angularDampingOverride=0.9999
    angularSpringConstant=5000
    bAngularSpring=true
    bOverrideAngularDamping=true
    bUseGravityOverride=true
    gravityScale=0.35
    simulationSpace=RootRelative
    alphaBoolBlend {blendOption=Linear}
    alphaScaleBiasClamp {scale=0.5}
    boundBone {boneName=dyn_cloak_rear_c_02}
    componentLinearAccScale=<0.05 0.05 0.05>
    componentLinearVelScale=<0.05 0.05 0.05>
    gravityOverride {z=-980}
    physicsBodyDefinitions [{
      boundBone {boneName=dyn_cloak_rear_c_02}
      boxExtents=<150 150 150>
      constraintSetup {angularLimitsMax=<0 15 60>, angularLimitsMin=<0 -60 -60>, angularTarget=<1 0 0>}
      localJointOffset=<0 0 0>
     }]
   }
   n20 = Anim Dynamics - Bone: dyn_cloak_rear_c_01(ComponentPose=@n19, Alpha=1, ExternalForce="0,0,0") {
    angularDampingOverride=0.9999
    angularSpringConstant=5000
    bAngularSpring=true
    bOverrideAngularDamping=true
    bUseGravityOverride=true
    gravityScale=0.35
    simulationSpace=RootRelative
    alphaBoolBlend {blendOption=Linear}
    alphaScaleBiasClamp {scale=0.25}
    boundBone {boneName=dyn_cloak_rear_c_01}
    componentLinearAccScale=<0.05 0.05 0.05>
    componentLinearVelScale=<0.05 0.05 0.05>
    gravityOverride {z=-980}
    physicsBodyDefinitions [{
      boundBone {boneName=dyn_cloak_rear_c_01}
      boxExtents=<150 150 150>
      constraintSetup {angularLimitsMax=<0 15 60>, angularLimitsMin=<0 -60 -60>, angularTarget=<1 0 0>}
      localJointOffset=<0 0 0>
     }]
   }
   n21 = Anim Dynamics - Bone: dyn_cloak_rear_r_05(ComponentPose=@n20, Alpha=1, ExternalForce="0,0,0") {
    angularDampingOverride=0.9999
    angularSpringConstant=5000
    bAngularSpring=true
    bOverrideAngularDamping=true
    bUseGravityOverride=true
    gravityScale=0.35
    simulationSpace=RootRelative
    alphaBoolBlend {blendOption=Linear}
    boundBone {boneName=dyn_cloak_rear_r_05}
    componentLinearAccScale=<0.05 0.05 0.05>
    componentLinearVelScale=<0.05 0.05 0.05>
    gravityOverride {z=-980}
    physicsBodyDefinitions [{
      boundBone {boneName=dyn_cloak_rear_r_05}
      boxExtents=<150 150 150>
      constraintSetup {angularLimitsMax=<0 15 60>, angularLimitsMin=<0 -60 -60>, angularTarget=<1 0 0>}
      localJointOffset=<0 0 0>
     }]
   }
   n22 = Anim Dynamics - Bone: dyn_cloak_rear_r_04(ComponentPose=@n21, Alpha=1, ExternalForce="0,0,0") {
    angularDampingOverride=0.9999
    angularSpringConstant=5000
    bAngularSpring=true
    bOverrideAngularDamping=true
    bUseGravityOverride=true
    gravityScale=0.35
    simulationSpace=RootRelative
    alphaBoolBlend {blendOption=Linear}
    alphaScaleBiasClamp {scale=0.75}
    boundBone {boneName=dyn_cloak_rear_r_04}
    componentLinearAccScale=<0.05 0.05 0.05>
    componentLinearVelScale=<0.05 0.05 0.05>
    gravityOverride {z=-980}
    physicsBodyDefinitions [{
      boundBone {boneName=dyn_cloak_rear_r_04}
      boxExtents=<150 150 150>
      constraintSetup {angularLimitsMax=<0 15 60>, angularLimitsMin=<0 -60 -60>, angularTarget=<1 0 0>}
      localJointOffset=<0 0 0>
     }]
   }
   n23 = Anim Dynamics - Bone: dyn_cloak_rear_r_03(ComponentPose=@n22, Alpha=1, ExternalForce="0,0,0") {
    angularDampingOverride=0.9999
    angularSpringConstant=5000
    bAngularSpring=true
    bOverrideAngularDamping=true
    bUseGravityOverride=true
    gravityScale=0.35
    simulationSpace=RootRelative
    alphaBoolBlend {blendOption=Linear}
    alphaScaleBiasClamp {scale=0.75}
    boundBone {boneName=dyn_cloak_rear_r_03}
    componentLinearAccScale=<0.05 0.05 0.05>
    componentLinearVelScale=<0.05 0.05 0.05>
    gravityOverride {z=-980}
    physicsBodyDefinitions [{
      boundBone {boneName=dyn_cloak_rear_r_03}
      boxExtents=<150 150 150>
      constraintSetup {angularLimitsMax=<0 15 60>, angularLimitsMin=<0 -60 -60>, angularTarget=<1 0 0>}
      localJointOffset=<0 0 0>
     }]
   }
   n24 = Anim Dynamics - Bone: dyn_cloak_rear_r_02(ComponentPose=@n23, Alpha=1, ExternalForce="0,0,0") {
    angularDampingOverride=0.9999
    angularSpringConstant=5000
    bAngularSpring=true
    bOverrideAngularDamping=true
    bUseGravityOverride=true
    gravityScale=0.35
    simulationSpace=RootRelative
    alphaBoolBlend {blendOption=Linear}
    alphaScaleBiasClamp {scale=0.5}
    boundBone {boneName=dyn_cloak_rear_r_02}
    componentLinearAccScale=<0.05 0.05 0.05>
    componentLinearVelScale=<0.05 0.05 0.05>
    gravityOverride {z=-980}
    physicsBodyDefinitions [{
      boundBone {boneName=dyn_cloak_rear_r_02}
      boxExtents=<150 150 150>
      constraintSetup {angularLimitsMax=<0 15 60>, angularLimitsMin=<0 -60 -60>, angularTarget=<1 0 0>}
      localJointOffset=<0 0 0>
     }]
   }
   n25 = Anim Dynamics - Bone: dyn_cloak_rear_r_01(ComponentPose=@n24, Alpha=1, ExternalForce="0,0,0") {
    angularDampingOverride=0.9999
    angularSpringConstant=5000
    bAngularSpring=true
    bOverrideAngularDamping=true
    bUseGravityOverride=true
    gravityScale=0.35
    simulationSpace=RootRelative
    alphaBoolBlend {blendOption=Linear}
    alphaScaleBiasClamp {scale=0.25}
    boundBone {boneName=dyn_cloak_rear_r_01}
    componentLinearAccScale=<0.05 0.05 0.05>
    componentLinearVelScale=<0.05 0.05 0.05>
    gravityOverride {z=-980}
    physicsBodyDefinitions [{
      boundBone {boneName=dyn_cloak_rear_r_01}
      boxExtents=<150 150 150>
      constraintSetup {angularLimitsMax=<0 15 60>, angularLimitsMin=<0 -60 -60>, angularTarget=<1 0 0>}
      localJointOffset=<0 0 0>
     }]
   }
   n26 = Anim Dynamics - Bone: dyn_cloak_side_r_05(ComponentPose=@n25, Alpha=1, ExternalForce="0,0,0") {
    angularDampingOverride=0.9999
    angularSpringConstant=5000
    bAngularSpring=true
    bOverrideAngularDamping=true
    bUseGravityOverride=true
    gravityScale=0.35
    simulationSpace=RootRelative
    alphaBoolBlend {blendOption=Linear}
    boundBone {boneName=dyn_cloak_side_r_05}
    componentLinearAccScale=<0.05 0.05 0.05>
    componentLinearVelScale=<0.05 0.05 0.05>
    gravityOverride {z=-980}
    physicsBodyDefinitions [{
      boundBone {boneName=dyn_cloak_side_r_05}
      boxExtents=<150 150 150>
      constraintSetup {angularLimitsMax=<0 15 60>, angularLimitsMin=<0 -60 -60>, angularTarget=<1 0 0>}
      localJointOffset=<0 0 0>
     }]
   }
   n27 = Anim Dynamics - Bone: dyn_cloak_side_r_04(ComponentPose=@n26, Alpha=1, ExternalForce="0,0,0") {
    angularDampingOverride=0.9999
    angularSpringConstant=5000
    bAngularSpring=true
    bOverrideAngularDamping=true
    bUseGravityOverride=true
    gravityScale=0.35
    simulationSpace=RootRelative
    alphaBoolBlend {blendOption=Linear}
    alphaScaleBiasClamp {scale=0.75}
    boundBone {boneName=dyn_cloak_side_r_04}
    componentLinearAccScale=<0.05 0.05 0.05>
    componentLinearVelScale=<0.05 0.05 0.05>
    gravityOverride {z=-980}
    physicsBodyDefinitions [{
      boundBone {boneName=dyn_cloak_side_r_04}
      boxExtents=<150 150 150>
      constraintSetup {angularLimitsMax=<0 15 60>, angularLimitsMin=<0 -60 -60>, angularTarget=<1 0 0>}
      localJointOffset=<0 0 0>
     }]
   }
   n28 = Anim Dynamics - Bone: dyn_cloak_side_r_03(ComponentPose=@n27, Alpha=1, ExternalForce="0,0,0") {
    angularDampingOverride=0.9999
    angularSpringConstant=5000
    bAngularSpring=true
    bOverrideAngularDamping=true
    bUseGravityOverride=true
    gravityScale=0.35
    simulationSpace=RootRelative
    alphaBoolBlend {blendOption=Linear}
    alphaScaleBiasClamp {scale=0.75}
    boundBone {boneName=dyn_cloak_side_r_03}
    componentLinearAccScale=<0.05 0.05 0.05>
    componentLinearVelScale=<0.05 0.05 0.05>
    gravityOverride {z=-980}
    physicsBodyDefinitions [{
      boundBone {boneName=dyn_cloak_side_r_03}
      boxExtents=<150 150 150>
      constraintSetup {angularLimitsMax=<0 15 60>, angularLimitsMin=<0 -60 -60>, angularTarget=<1 0 0>}
      localJointOffset=<0 0 0>
     }]
   }
   n29 = Anim Dynamics - Bone: dyn_cloak_side_r_02(ComponentPose=@n28, Alpha=1, ExternalForce="0,0,0") {
    angularDampingOverride=0.9999
    angularSpringConstant=5000
    bAngularSpring=true
    bOverrideAngularDamping=true
    bUseGravityOverride=true
    gravityScale=0.35
    simulationSpace=RootRelative
    alphaBoolBlend {blendOption=Linear}
    alphaScaleBiasClamp {scale=0.5}
    boundBone {boneName=dyn_cloak_side_r_02}
    componentLinearAccScale=<0.05 0.05 0.05>
    componentLinearVelScale=<0.05 0.05 0.05>
    gravityOverride {z=-980}
    physicsBodyDefinitions [{
      boundBone {boneName=dyn_cloak_side_r_02}
      boxExtents=<150 150 150>
      constraintSetup {angularLimitsMax=<0 15 60>, angularLimitsMin=<0 -60 -60>, angularTarget=<1 0 0>}
      localJointOffset=<0 0 0>
     }]
   }
   n30 = Anim Dynamics - Bone: dyn_cloak_side_r_01(ComponentPose=@n29, Alpha=1, ExternalForce="0,0,0") {
    angularDampingOverride=0.9999
    angularSpringConstant=5000
    bAngularSpring=true
    bOverrideAngularDamping=true
    bUseGravityOverride=true
    gravityScale=0.35
    simulationSpace=RootRelative
    alphaBoolBlend {blendOption=Linear}
    alphaScaleBiasClamp {scale=0.25}
    boundBone {boneName=dyn_cloak_side_r_01}
    componentLinearAccScale=<0.05 0.05 0.05>
    componentLinearVelScale=<0.05 0.05 0.05>
    gravityOverride {z=-980}
    physicsBodyDefinitions [{
      boundBone {boneName=dyn_cloak_side_r_01}
      boxExtents=<150 150 150>
      constraintSetup {angularLimitsMax=<0 15 60>, angularLimitsMin=<0 -60 -60>, angularTarget=<1 0 0>}
      localJointOffset=<0 0 0>
     }]
   }
   n31 = Anim Dynamics - Bone: dyn_cloak_front_r_05(ComponentPose=@n30, Alpha=1, ExternalForce="0,0,0") {
    angularDampingOverride=0.9999
    angularSpringConstant=5000
    bAngularSpring=true
    bOverrideAngularDamping=true
    bUseGravityOverride=true
    gravityScale=0.35
    simulationSpace=RootRelative
    alphaBoolBlend {blendOption=Linear}
    boundBone {boneName=dyn_cloak_front_r_05}
    componentLinearAccScale=<0.05 0.05 0.05>
    componentLinearVelScale=<0.05 0.05 0.05>
    gravityOverride {z=-980}
    physicsBodyDefinitions [{
      boundBone {boneName=dyn_cloak_front_r_05}
      boxExtents=<150 150 150>
      constraintSetup {angularLimitsMax=<0 15 60>, angularLimitsMin=<0 -60 -60>, angularTarget=<1 0 0>}
      localJointOffset=<0 0 0>
     }]
   }
   n32 = Anim Dynamics - Bone: dyn_cloak_front_r_04(ComponentPose=@n31, Alpha=1, ExternalForce="0,0,0") {
    angularDampingOverride=0.9999
    angularSpringConstant=5000
    bAngularSpring=true
    bOverrideAngularDamping=true
    bUseGravityOverride=true
    gravityScale=0.35
    simulationSpace=RootRelative
    alphaBoolBlend {blendOption=Linear}
    alphaScaleBiasClamp {scale=0.75}
    boundBone {boneName=dyn_cloak_front_r_04}
    componentLinearAccScale=<0.05 0.05 0.05>
    componentLinearVelScale=<0.05 0.05 0.05>
    gravityOverride {z=-980}
    physicsBodyDefinitions [{
      boundBone {boneName=dyn_cloak_front_r_04}
      boxExtents=<150 150 150>
      constraintSetup {angularLimitsMax=<0 15 60>, angularLimitsMin=<0 -60 -60>, angularTarget=<1 0 0>}
      localJointOffset=<0 0 0>
     }]
   }
   n33 = Anim Dynamics - Bone: dyn_cloak_front_r_03(ComponentPose=@n32, Alpha=1, ExternalForce="0,0,0") {
    angularDampingOverride=0.9999
    angularSpringConstant=5000
    bAngularSpring=true
    bOverrideAngularDamping=true
    bUseGravityOverride=true
    gravityScale=0.35
    simulationSpace=RootRelative
    alphaBoolBlend {blendOption=Linear}
    alphaScaleBiasClamp {scale=0.75}
    boundBone {boneName=dyn_cloak_front_r_03}
    componentLinearAccScale=<0.05 0.05 0.05>
    componentLinearVelScale=<0.05 0.05 0.05>
    gravityOverride {z=-980}
    physicsBodyDefinitions [{
      boundBone {boneName=dyn_cloak_front_r_03}
      boxExtents=<150 150 150>
      constraintSetup {angularLimitsMax=<0 15 60>, angularLimitsMin=<0 -60 -60>, angularTarget=<1 0 0>}
      localJointOffset=<0 0 0>
     }]
   }
   n34 = Anim Dynamics - Bone: dyn_cloak_front_r_02(ComponentPose=@n33, Alpha=1, ExternalForce="0,0,0") {
    angularDampingOverride=0.9999
    angularSpringConstant=5000
    bAngularSpring=true
    bOverrideAngularDamping=true
    bUseGravityOverride=true
    gravityScale=0.35
    simulationSpace=RootRelative
    alphaBoolBlend {blendOption=Linear}
    alphaScaleBiasClamp {scale=0.5}
    boundBone {boneName=dyn_cloak_front_r_02}
    componentLinearAccScale=<0.05 0.05 0.05>
    componentLinearVelScale=<0.05 0.05 0.05>
    gravityOverride {z=-980}
    physicsBodyDefinitions [{
      boundBone {boneName=dyn_cloak_front_r_02}
      boxExtents=<150 150 150>
      constraintSetup {angularLimitsMax=<0 15 60>, angularLimitsMin=<0 -60 -60>, angularTarget=<1 0 0>}
      localJointOffset=<0 0 0>
     }]
   }
   n35 = Anim Dynamics - Bone: dyn_cloak_front_r_01(ComponentPose=@n34, Alpha=1, ExternalForce="0,0,0") {
    angularDampingOverride=0.9999
    angularSpringConstant=5000
    bAngularSpring=true
    bOverrideAngularDamping=true
    bUseGravityOverride=true
    gravityScale=0.35
    simulationSpace=RootRelative
    alphaBoolBlend {blendOption=Linear}
    alphaScaleBiasClamp {scale=0.25}
    boundBone {boneName=dyn_cloak_front_r_01}
    componentLinearAccScale=<0.05 0.05 0.05>
    componentLinearVelScale=<0.05 0.05 0.05>
    gravityOverride {z=-980}
    physicsBodyDefinitions [{
      boundBone {boneName=dyn_cloak_front_r_01}
      boxExtents=<150 150 150>
      constraintSetup {angularLimitsMax=<0 15 60>, angularLimitsMin=<0 -60 -60>, angularTarget=<1 0 0>}
      localJointOffset=<0 0 0>
     }]
   }
   Output Pose(Result=Component To Local(ComponentPose=@n35))
   Save cached pose 'PreSim'(Pose=Copy Pose From Mesh() {bUseAttachedParent=true}) {cacheName=PreSim}
   n36 = Get End Rand Rot() -> (End Rand Rot:Rotator)
   n37 = Get Rand Rot 1() -> (Rand Rot 1:Rotator)
   n38 = Get Rand Rot 1 scaled() -> (Rand Rot 1 scaled:Rotator)
   n39 = "Transform (Modify) Bone - Bone: dyn_cloak_front_r_03"(ComponentPose=Local To Component(LocalPose=Use cached pose 'PreSim'()), Alpha=1, Translation="0,0,0", Rotation=@n38, Scale="1,1,1") {rotationMode=BMM_Additive, rotationSpace=BCS_BoneSpace, alphaBoolBlend {blendOption=Linear}, boneToModify {boneName=dyn_cloak_front_r_03}}
   n40 = "Transform (Modify) Bone - Bone: dyn_cloak_side_r_03"(ComponentPose=@n39, Alpha=1, Translation="0,0,0", Rotation=@n38, Scale="1,1,1") {rotationMode=BMM_Additive, rotationSpace=BCS_BoneSpace, alphaBoolBlend {blendOption=Linear}, boneToModify {boneName=dyn_cloak_side_r_03}}
   n41 = "Transform (Modify) Bone - Bone: dyn_cloak_rear_r_03"(ComponentPose=@n40, Alpha=1, Translation="0,0,0", Rotation=@n38, Scale="1,1,1") {rotationMode=BMM_Additive, rotationSpace=BCS_BoneSpace, alphaBoolBlend {blendOption=Linear}, boneToModify {boneName=dyn_cloak_rear_r_03}}
   n42 = Get Rand Rot 2 scaled() -> (Rand Rot 2 scaled:Rotator)
   n43 = "Transform (Modify) Bone - Bone: dyn_cloak_rear_c_04"(ComponentPose=@n41, Alpha=1, Translation="0,0,0", Rotation=@n42, Scale="1,1,1") {rotationMode=BMM_Additive, rotationSpace=BCS_BoneSpace, alphaBoolBlend {blendOption=Linear}, boneToModify {boneName=dyn_cloak_rear_c_04}}
   n44 = "Transform (Modify) Bone - Bone: dyn_cloak_rear_l_03"(ComponentPose=@n43, Alpha=1, Translation="0,0,0", Rotation=@n42, Scale="1,1,1") {rotationMode=BMM_Additive, rotationSpace=BCS_BoneSpace, alphaBoolBlend {blendOption=Linear}, boneToModify {boneName=dyn_cloak_rear_l_03}}
   n45 = "Transform (Modify) Bone - Bone: dyn_cloak_side_l_03"(ComponentPose=@n44, Alpha=1, Translation="0,0,0", Rotation=@n42, Scale="1,1,1") {rotationMode=BMM_Additive, rotationSpace=BCS_BoneSpace, alphaBoolBlend {blendOption=Linear}, boneToModify {boneName=dyn_cloak_side_l_03}}
   n46 = "Transform (Modify) Bone - Bone: dyn_cloak_front_l_03"(ComponentPose=@n45, Alpha=1, Translation="0,0,0", Rotation=@n42, Scale="1,1,1") {rotationMode=BMM_Additive, rotationSpace=BCS_BoneSpace, alphaBoolBlend {blendOption=Linear}, boneToModify {boneName=dyn_cloak_front_l_03}}
   n47 = "Transform (Modify) Bone - Bone: dyn_cloak_front_r_04"(ComponentPose=@n46, Alpha=1, Translation="0,0,0", Rotation=@n37, Scale="1,1,1") {rotationMode=BMM_Additive, rotationSpace=BCS_BoneSpace, alphaBoolBlend {blendOption=Linear}, boneToModify {boneName=dyn_cloak_front_r_04}}
   n48 = "Transform (Modify) Bone - Bone: dyn_cloak_side_r_04"(ComponentPose=@n47, Alpha=1, Translation="0,0,0", Rotation=@n37, Scale="1,1,1") {rotationMode=BMM_Additive, rotationSpace=BCS_BoneSpace, alphaBoolBlend {blendOption=Linear}, boneToModify {boneName=dyn_cloak_side_r_04}}
   n49 = "Transform (Modify) Bone - Bone: dyn_cloak_rear_r_04"(ComponentPose=@n48, Alpha=1, Translation="0,0,0", Rotation=@n37, Scale="1,1,1") {rotationMode=BMM_Additive, rotationSpace=BCS_BoneSpace, alphaBoolBlend {blendOption=Linear}, boneToModify {boneName=dyn_cloak_rear_r_04}}
   n50 = "Transform (Modify) Bone - Bone: dyn_cloak_rear_c_04"(ComponentPose=@n49, Alpha=1, Translation="0,0,0", Rotation=@n37, Scale="1,1,1") {rotationMode=BMM_Additive, rotationSpace=BCS_BoneSpace, alphaBoolBlend {blendOption=Linear}, boneToModify {boneName=dyn_cloak_rear_c_04}}
   n51 = Get Rand Rot 2() -> (Rand Rot 2:Rotator)
   n52 = "Transform (Modify) Bone - Bone: dyn_cloak_rear_l_04"(ComponentPose=@n50, Alpha=1, Translation="0,0,0", Rotation=@n51, Scale="1,1,1") {rotationMode=BMM_Additive, rotationSpace=BCS_BoneSpace, alphaBoolBlend {blendOption=Linear}, boneToModify {boneName=dyn_cloak_rear_l_04}}
   n53 = "Transform (Modify) Bone - Bone: dyn_cloak_side_l_04"(ComponentPose=@n52, Alpha=1, Translation="0,0,0", Rotation=@n51, Scale="1,1,1") {rotationMode=BMM_Additive, rotationSpace=BCS_BoneSpace, alphaBoolBlend {blendOption=Linear}, boneToModify {boneName=dyn_cloak_side_l_04}}
   n54 = "Transform (Modify) Bone - Bone: dyn_cloak_front_l_04"(ComponentPose=@n53, Alpha=1, Translation="0,0,0", Rotation=@n51, Scale="1,1,1") {rotationMode=BMM_Additive, rotationSpace=BCS_BoneSpace, alphaBoolBlend {blendOption=Linear}, boneToModify {boneName=dyn_cloak_front_l_04}}
   n55 = "Transform (Modify) Bone - Bone: dyn_cloak_front_r_05"(ComponentPose=@n54, Alpha=1, Translation="0,0,0", Rotation=@n36, Scale="1,1,1") {rotationMode=BMM_Additive, rotationSpace=BCS_BoneSpace, alphaBoolBlend {blendOption=Linear}, boneToModify {boneName=dyn_cloak_front_r_05}}
   n56 = "Transform (Modify) Bone - Bone: dyn_cloak_side_r_05"(ComponentPose=@n55, Alpha=1, Translation="0,0,0", Rotation=@n36, Scale="1,1,1") {rotationMode=BMM_Additive, rotationSpace=BCS_BoneSpace, alphaBoolBlend {blendOption=Linear}, boneToModify {boneName=dyn_cloak_side_r_05}}
   n57 = "Transform (Modify) Bone - Bone: dyn_cloak_rear_r_05"(ComponentPose=@n56, Alpha=1, Translation="0,0,0", Rotation=@n36, Scale="1,1,1") {rotationMode=BMM_Additive, rotationSpace=BCS_BoneSpace, alphaBoolBlend {blendOption=Linear}, boneToModify {boneName=dyn_cloak_rear_r_05}}
   n58 = "Transform (Modify) Bone - Bone: dyn_cloak_rear_c_05"(ComponentPose=@n57, Alpha=1, Translation="0,0,0", Rotation=@n36, Scale="1,1,1") {rotationMode=BMM_Additive, rotationSpace=BCS_BoneSpace, alphaBoolBlend {blendOption=Linear}, boneToModify {boneName=dyn_cloak_rear_c_05}}
   n59 = "Transform (Modify) Bone - Bone: dyn_cloak_rear_l_05"(ComponentPose=@n58, Alpha=1, Translation="0,0,0", Rotation=@n36, Scale="1,1,1") {rotationMode=BMM_Additive, rotationSpace=BCS_BoneSpace, alphaBoolBlend {blendOption=Linear}, boneToModify {boneName=dyn_cloak_rear_l_05}}
   n60 = "Transform (Modify) Bone - Bone: dyn_cloak_side_l_05"(ComponentPose=@n59, Alpha=1, Translation="0,0,0", Rotation=@n36, Scale="1,1,1") {rotationMode=BMM_Additive, rotationSpace=BCS_BoneSpace, alphaBoolBlend {blendOption=Linear}, boneToModify {boneName=dyn_cloak_side_l_05}}
   n61 = "Transform (Modify) Bone - Bone: dyn_cloak_front_l_05"(ComponentPose=@n60, Alpha=1, Translation="0,0,0", Rotation=@n36, Scale="1,1,1") {rotationMode=BMM_Additive, rotationSpace=BCS_BoneSpace, alphaBoolBlend {blendOption=Linear}, boneToModify {boneName=dyn_cloak_front_l_05}}
   Save cached pose 'PostRotation'(Pose=Component To Local(ComponentPose=@n61)) {cacheName=PostRotation}
  ]}
 EventGraph {nodes [
   Event BlueprintUpdateAnimation()
   > n1 = Sequence!(execute) -> (then_0, then_1, then_2:Exec)
   n2 = GetGameTimeInSeconds()
   > n3 = RandomRotatorFunction!(
    execute
    GameTimeInSeconds:double=@n2
    PerlinScale:double=1
    Frequency:Rotator="0.2,10,0"
    Amplitude:Rotator="15,1,0"
    Offset:Rotator="0.5,0,0"
    TimeOffset:double=0
   ) -> (then, RandomRotator:Rotator)
   n4 = BreakRotator(InRot=@n3)
   > n5 = Set Rand Rot 1(Rand Rot 1:Rotator=MakeRotator(Roll=@n4, Pitch="Absolute (Float)"(A=@n4.Pitch), Yaw=@n4.Yaw)) -> (Output_Get:Rotator)
   n6 = BreakRotator(InRot=@n5)
   > Set Rand Rot 1 scaled(Rand Rot 1 scaled:Rotator=MakeRotator(Roll=@n6, Pitch=float * float(A=@n6.Pitch, B=0.6) {is=title K2Node_PromotableOperator}, Yaw=float * float(A=@n6.Yaw, B=0.6))) -> (Output_Get:Rotator)
   n7 = RandomRotatorFunction(execute=@n1.then_1, GameTimeInSeconds=@n2, PerlinScale=1, Frequency="0.3,8,0", Amplitude="12,1,0", Offset="0.5,0,0", TimeOffset=0)
   n8 = BreakRotator(InRot=@n7)
   > n9 = Set Rand Rot 2(Rand Rot 2:Rotator=MakeRotator(Roll=@n8, Pitch="Absolute (Float)"(A=@n8.Pitch), Yaw=@n8.Yaw)) -> (Output_Get:Rotator)
   n10 = BreakRotator(InRot=@n9)
   > Set Rand Rot 2 scaled(Rand Rot 2 scaled:Rotator=MakeRotator(Roll=@n10, Pitch=float * float(A=@n10.Pitch, B=0.6), Yaw=float * float(A=@n10.Yaw, B=0.6))) -> (Output_Get:Rotator)
   n11 = RandomRotatorFunction(execute=@n1.then_2, GameTimeInSeconds=@n2, PerlinScale=1, Frequency="0.2,0,0", Amplitude="1,0,0", Offset="1,0,0", TimeOffset=0)
   n12 = BreakRotator(InRot=@n11)
   > Set End Rand Rot(End Rand Rot:Rotator=MakeRotator(Roll=@n12, Pitch="Absolute (Float)"(A=@n12.Pitch), Yaw=@n12.Yaw)) -> (Output_Get:Rotator)
   TryGetPawnOwner()
  ]}
]
variables [
 End Rand Rot {type=Rotator}
 Force Vector {type=Vector}
 Rand Rot 1 {type=Rotator}
 Rand Rot 1 scaled {type=Rotator}
 Rand Rot 2 {type=Rotator}
 Rand Rot 2 scaled {type=Rotator}
]
